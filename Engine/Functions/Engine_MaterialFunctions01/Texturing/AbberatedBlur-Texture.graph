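# Library material function that blurs a 2D texture along a given axis with chromatic aberration.
# A Custom HLSL node loops over Steps samples, spanning Distance in UV space along Vector, centered on the input UV.
# Position-dependent weights favor red, then green, then blue along the path to split colors.
# The weighted sum is divided by the sample count; the loop counter and accumulator are declared without initial values.
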MaterialFunction /Engine/Functions/Engine_MaterialFunctions01/Texturing/AbberatedBlur-Texture
graph {
 properties {
  bExposeToLibrary=true
  description=Blurs a Texture along a specified 2D Axis with chromatic abberation.
  libraryCategoriesText [Misc,Texturing]
 }
 nodes [
  MaterialExpressionFunctionInput_5 = FunctionInput() {description=Supply the Texture to Blur as a TextureObject node, inputName=TextureObject, inputType=FunctionInput_Texture2D}
  MaterialExpressionFunctionInput_8 = FunctionInput(Preview=TextureCoordinate()) {bUsePreviewValueAsDefault=true, description=UVs for blurring, inputName=UV, inputType=FunctionInput_Vector2, sortPriority=1}
  MaterialExpressionFunctionInput_15 = FunctionInput() {
   bUsePreviewValueAsDefault=true
   description=V2 vector for blur direction
   inputName=Vector
   inputType=FunctionInput_Vector2
   sortPriority=2
   previewValue=<1 0 0 1>
  }
  MaterialExpressionFunctionInput_14 = FunctionInput() {
   bUsePreviewValueAsDefault=true
   description=Total Distance of Blur in UV Space
   inputName=Distance
   inputType=FunctionInput_Scalar
   sortPriority=3
   previewValue=<0.05 0 0 1>
  }
  MaterialExpressionFunctionInput_11 = FunctionInput() {
   bUsePreviewValueAsDefault=true
   description=Number of steps
   inputName=Steps
   inputType=FunctionInput_Scalar
   sortPriority=4
   previewValue=<16 0 0 1>
  }
  n1 = Custom!(
   Tex=@MaterialExpressionFunctionInput_5
   UV=@MaterialExpressionFunctionInput_8
   Vector=@MaterialExpressionFunctionInput_15
   Distance=@MaterialExpressionFunctionInput_14
   Steps=@MaterialExpressionFunctionInput_11
  ) {
   code="float3 CurColor;\r\nfloat3 newSample;\r\n\r\nint i;\r\nSteps = max(Steps, 1);\r\nfloat StepSize = Distance / (int) Steps;\r\nfloat2 NewUV = UV - Vector*0.5*Distance;\r\nwhile ( i < (int) Steps)\r\n{\r\nfloat3 newSample=Texture2DSample(Tex,TexSampler,NewUV);\r\nfloat3 CurrentAbberationValue=saturate(1.0-abs(3*(i/Steps)-float4(0.0,1.0,2.0,3.0)));\r\nCurColor += newSample*CurrentAbberationValue;\r\nNewUV.x += (StepSize * Vector.x);\r\nNewUV.y += (StepSize * Vector.y);\r\ni++;\r\n}\r\nCurColor = CurColor / i;\r\nreturn CurColor;"
   inputs [
    {
     inputName=Tex
     input {expression=&self:MaterialExpressionFunctionInput_5}
    }
    {
     inputName=UV
     input {expression=&self:MaterialExpressionFunctionInput_8}
    }
    {
     inputName=Vector
     input {expression=&self:MaterialExpressionFunctionInput_15}
    }
    {
     inputName=Distance
     input {expression=&self:MaterialExpressionFunctionInput_14}
    }
    {
     inputName=Steps
     input {expression=&self:MaterialExpressionFunctionInput_11}
    }
   ]
  }
  FunctionOutput(None=@n1) {outputName=Result With Abberation}
 ]
}
