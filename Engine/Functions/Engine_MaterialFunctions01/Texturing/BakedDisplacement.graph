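# BakedDisplacement is a library-exposed material function that turns a baked scalar displacement map value into an offset for world displacement.
# A scalar input, left at its default name, serves as the alpha that interpolates between the Baked_Min and Baked_Max inputs, remapping the value to that range.
# The result is multiplied by the world-space vertex normal to give the output displacement vector.
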MaterialFunction /Engine/Functions/Engine_MaterialFunctions01/Texturing/BakedDisplacement
graph {
 properties {bExposeToLibrary=true, description=Set up a baked scalar displacement map to be used with WorldDispalcement}
 nodes [
  n1 = FunctionInput() {bUsePreviewValueAsDefault=true, inputName=Baked_Min, inputType=FunctionInput_Scalar, sortPriority=1}
  n2 = FunctionInput() {bUsePreviewValueAsDefault=true, inputName=Baked_Max, inputType=FunctionInput_Scalar, sortPriority=2}
  n3 = FunctionInput() {inputType=FunctionInput_Scalar}
  FunctionOutput(None=Multiply(A=LinearInterpolate(A=@n1, B=@n2, Alpha=@n3), B=VertexNormalWS()))
 ]
}
