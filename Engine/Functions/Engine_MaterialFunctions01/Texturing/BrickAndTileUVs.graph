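# Material function that builds semi-procedural UVs for brick and tile patterned textures, with optional 90-degree axis swap.
# Scales UVs by tile counts, then applies a sine of the row index to stagger tiles.
# A custom HLSL hash of the distorted tile index gives pseudo-random per-tile offsets for diffuse and macro UVs.
# Also outputs those offsets alone and tile-space UVs; an older manual offset calculation remains unconnected.
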MaterialFunction /Engine/Functions/Engine_MaterialFunctions01/Texturing/BrickAndTileUVs
graph {
 properties {
  bExposeToLibrary=true
  description=Bricks and Tile UVs create a semi procedural UV offset pattern useful for the creation of patterned textures.
  libraryCategoriesText [Texturing]
 }
 nodes [
  n1 = FunctionInput() {
   description=Allows you to distort the random tile edge pattern.
   inputName=Pattern Distortion
   inputType=FunctionInput_Vector2
   sortPriority=6
  }
  n2 = Multiply(A=@n1)
  n3 = FunctionInput() {bUsePreviewValueAsDefault=true, inputName=Tiles X, inputType=FunctionInput_Scalar, previewValue=<3 0 0 1>}
  n4 = FunctionInput() {bUsePreviewValueAsDefault=true, inputName=Tiles Y, inputType=FunctionInput_Scalar, sortPriority=1, previewValue=<8 0 0 1>}
  n5 = AppendVector(A=@n3, B=@n4)
  n6 = FunctionInput(Preview=TextureCoordinate()) {bUsePreviewValueAsDefault=true, inputName=UV, inputType=FunctionInput_Vector2, sortPriority=-10}
  n7 = StaticSwitch(
   True=MaterialFunctionCall(XY=@n6) {materialFunction=&Swizzle}.YX
   False=@n6
   Value=FunctionInput(Preview=StaticBool()) {bUsePreviewValueAsDefault=true, inputName=Rotate 90, inputType=FunctionInput_StaticBool, sortPriority=5}
  ) {desc=Raw coords}
  n8 = Multiply(A=@n5, B=@n7)
  n9 = Add(A=@n8, B=@n1)
  n10 = Constant() {desc=0, r=0.5}
  n11 = Constant() {desc=Number of stagger steps, r=2}
  n12 = ComponentMask(None=Multiply(A=Floor(None=@n9) {desc=Tile Index}, B=Divide(A=@n10, B=@n11))) {g=true}
  n13 = Constant()
  n14 = Multiply(
   A=FunctionInput(Preview=Constant() {desc=Tile Shift, r=0.1}) {bUsePreviewValueAsDefault=true, inputName=Tile Shift, inputType=FunctionInput_Scalar, sortPriority=2}
   B=Sine(None=AppendVector(A=@n12, B=@n13)) {desc=Stagger sine}
  )
  n15 = Frac(None=Add(A=Divide(A=@n14, B=@n3), B=@n7))
  MaterialExpressionFloor_2 = Floor(None=Add(A=@n2, B=Multiply(A=@n5, B=@n15) {desc=Wrapped Coords}) {desc=Add Warble}) {desc=Tile Index}
  n16 = Multiply(A=@MaterialExpressionFloor_2) {constB=0.31}
  n17 = ComponentMask(None=@n16) {r=true}
  n18 = Constant2Vector() {g=0.411}
  Add(A=Add(A=@n16, B=Multiply(A=@n17, B=@n18)), B=Multiply(A=ComponentMask(None=@n16) {g=true}, B=Constant2Vector() {r=0.13})) {desc=Old Manual ver}
  n19 = Multiply(
   A=FunctionInput(Preview=Constant() {r=1}) {bUsePreviewValueAsDefault=true, inputName=Texture Tiling, inputType=FunctionInput_Scalar, sortPriority=3}
   B=@n15
  ) {desc=Wrapped Texture}
  n20 = Custom!(pos=@MaterialExpressionFloor_2) {
   code="float2 noise = dot(pos.xy * pos.xy, float2(12.581057, 11.129465));\r\nnoise.y = dot(pos.xy * pos.xy/5, float2(46.538326, 11.057759));\r\n\r\nreturn frac(noise);"
   desc="New, based on Martin's Code"
   description=Pseudo Random
   outputType=CMOT_Float2
   inputs [{inputName=pos, input {expression=&self:MaterialExpressionFloor_2}}]
  }
  Add(A=@n19, B=@n20)
  n21 = Multiply(
   A=FunctionInput(Preview=Constant() {r=0.18}) {bUsePreviewValueAsDefault=true, inputName=Macro Tiling, inputType=FunctionInput_Scalar, sortPriority=4}
   B=@n15
  ) {desc=Macro Wrapped}
  FunctionOutput(None=Add(A=Add(A=@n20, B=@n21) {desc=Macro}) {constB=0, desc=Macro UVs}) {outputName=MacroUVs, sortPriority=1}
  FunctionOutput(None=@n20) {outputName=Macro Offsets only, sortPriority=2}
  n22 = Multiply(A=@n8) {desc=Tiled UVs}
  FunctionOutput(None=Add(A=Add(A=@n2, B=@n22), B=@n14) {desc=Mask UVs}) {desc=Returns UVs that tile 0-1 for each tile then 1-2 the next etc no offsets., outputName=TileUVs, sortPriority=2}
  FunctionOutput(None=Add(A=@n19, B=@n20) {constB=0, desc=Diffuse UVs}) {outputName=DiffuseUVs}
 ]
}
