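# Material function that tiles a texture with a cylindrical projection around the object's center, outputting it alone or with planar caps.
# U comes from the XY angle around the object position via VectorToRadialValue, while V is world Z divided by a height input.
# The capped output blends in world XY sampling at the same scale, using contrasted absolute Z of the transformed Normal input.
# World position excludes shader offsets.
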MaterialFunction /Engine/Functions/Engine_MaterialFunctions01/Texturing/CylindricalUVs
graph {
 properties {
  bExposeToLibrary=true
  description=Tiles a texture using Cylindrical UVS centered around Object Center
  libraryCategoriesText [Misc,Texturing]
 }
 nodes [
  n1 = WorldPosition() {worldPositionShaderOffset=WPT_ExcludeAllShaderOffsets}
  n2 = ComponentMask(None=Subtract(A=@n1.XYZ, B=ObjectPositionWS())) {g=true, r=true}
  n3 = Divide(A=@n1.XYZ, B=FunctionInput() {description=UVHeight, inputType=FunctionInput_Scalar, previewValue=<512 0 0 1>})
  n4 = FunctionInput(Preview=TextureObject() {texture=/Engine/EngineMaterials/DefaultDiffuse}) {inputName=TextureObject, inputType=FunctionInput_Texture2D}
  n5 = TextureSample(
   UVs=AppendVector(
    A=MaterialFunctionCall(Vector or UVs=@n2) {materialFunction=&VectorToRadialValue}.Vector Converted to Angle
    B=ComponentMask(None=@n3) {b=true}
   )
   Tex=@n4
  ) {texture=&DefaultDiffuse}
  FunctionOutput(None=@n5.RGB) {outputName=Cylinder Projection}
  n6 = ComponentMask(None=@n3) {g=true, r=true}
  n7 = FunctionInput() {inputName=Normal, previewValue=<0 0 1 1>}
  n8 = Abs(None=ComponentMask(None=Transform(None=@n7)) {b=true})
  n9 = Constant() {r=1}
  FunctionOutput(None=LinearInterpolate(
    A=@n5.RGB
    B=TextureSample(UVs=@n6, Tex=@n4).RGB
    Alpha=MaterialFunctionCall(In=@n8, Contrast=@n9) {materialFunction=&CheapContrast}.Result
   )) {outputName=Cylinder Projection w Top}
 ]
}
