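# MosaicUVs is a material function that quantizes UVs to give textures a mosaic look of blocks or large pixels.
# It multiplies the UVs by a scalar Tiles input, applies Ceil, then divides by Tiles, snapping each coordinate to the upper edge of its cell.
# Both inputs have defaults: the first texture coordinate set for UVs and a preview value for Tiles.
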MaterialFunction /Engine/Functions/Engine_MaterialFunctions01/Texturing/MosaicUVs
graph {
 properties {description="Creates a mosaic style effect where the texture will be blocks, or large pixels", libraryCategoriesText [Misc,Texturing]}
 nodes [
  n1 = FunctionInput() {bUsePreviewValueAsDefault=true, inputName=Tiles, inputType=FunctionInput_Scalar, sortPriority=1, previewValue=<16 0 0 1>}
  n2 = Multiply(A=FunctionInput(Preview=TextureCoordinate()) {bUsePreviewValueAsDefault=true, inputName=UVs, inputType=FunctionInput_Vector2}, B=@n1)
  FunctionOutput(None=Divide(A=Ceil(None=@n2), B=@n1))
 ]
}
