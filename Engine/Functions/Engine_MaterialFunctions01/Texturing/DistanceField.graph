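# Material function that turns a sampled distance field value into an anti-aliased or soft-edged mask.
# It takes a scalar DistanceField input and an optional EdgeSoftness, which offsets 0.5 down and up to form the edge bounds.
# A Custom HLSL node applies smoothstep between those bounds to the distance value and outputs the result.
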MaterialFunction /Engine/Functions/Engine_MaterialFunctions01/Texturing/DistanceField
graph {
 properties {description=Reads a Distance field texture, libraryCategoriesText [Misc,Texturing]}
 nodes [
  MaterialExpressionFunctionInput_2 = FunctionInput() {inputName=DistanceField, inputType=FunctionInput_Scalar}
  n1 = FunctionInput() {bUsePreviewValueAsDefault=true, inputName=EdgeSoftness, inputType=FunctionInput_Scalar}
  MaterialExpressionSubtract_0 = Subtract(A=Constant() {r=0.5}, B=@n1)
  MaterialExpressionAdd_1 = Add(A=@n1, B=Constant() {r=0.5})
  n2 = Custom!(varA=@MaterialExpressionFunctionInput_2, varMin=@MaterialExpressionSubtract_0, varMax=@MaterialExpressionAdd_1) {
   code="return smoothstep(varMin,varMax,varA);"
   inputs [
    {inputName=varA, input {expression=&self:MaterialExpressionFunctionInput_2}}
    {inputName=varMin, input {expression=&self:MaterialExpressionSubtract_0}}
    {inputName=varMax, input {expression=&self:MaterialExpressionAdd_1}}
   ]
  }
  FunctionOutput(None=@n2)
 ]
}
