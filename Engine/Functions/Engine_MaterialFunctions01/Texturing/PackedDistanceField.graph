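# Material function, exposed to the library, that reads one distance field image from a texture packed as a subUV grid.
# A SubUV_Function call samples the grid cell chosen by the rounded-up image number, using the given UVs and image count.
# Its RGB feeds a DistanceField function with an edge softness input, and that result is the single output.
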MaterialFunction /Engine/Functions/Engine_MaterialFunctions01/Texturing/PackedDistanceField
graph {
 properties {bExposeToLibrary=true, description=Reads Distance Field textures packed as a subUV, libraryCategoriesText [Misc,Texturing]}
 nodes [
  n1 = FunctionInput() {
   bUsePreviewValueAsDefault=true
   description=DistanceField Texture packed as SubUV
   inputName=PackedDistanceField
   inputType=FunctionInput_Texture2D
   sortPriority=-1
   previewValue=<1 1 0 1>
  }
  n2 = FunctionInput(Preview=TextureCoordinate()) {bUsePreviewValueAsDefault=true, inputName=UVs, inputType=FunctionInput_Vector2, sortPriority=1}
  n3 = FunctionInput() {
   bUsePreviewValueAsDefault=true
   description=How many images horizontally and vertically
   inputName=ImageCount
   inputType=FunctionInput_Vector2
   sortPriority=2
   previewValue=<1 1 0 1>
  }
  n4 = FunctionInput() {description="Which image to grab, left to right top to bottom", inputName=ImageNumber, inputType=FunctionInput_Scalar, sortPriority=3}
  n5 = Ceil(None=@n4)
  n6 = FunctionInput() {
   bUsePreviewValueAsDefault=true
   description=How soft the distance field edge should be
   inputName=EdgeSoftness
   inputType=FunctionInput_Scalar
   sortPriority=3
  }
  FunctionOutput(None=MaterialFunctionCall!(
    DistanceField=MaterialFunctionCall!(Texture=@n1, UVs=@n2, SubImages=@n3, Frame=@n5) {materialFunction=/Engine/Functions/Engine_MaterialFunctions03/Texturing/SubUV_Function}.RGB
    EdgeSoftness=@n6
   ) {materialFunction=&/DistanceField}.Result)
 ]
}
