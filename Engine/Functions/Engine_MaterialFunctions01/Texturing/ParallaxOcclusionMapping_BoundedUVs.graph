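# Material function for parallax occlusion mapping confined to a UV rectangle (ClampMinMax), with a mask for rays that miss.
# Height Ratio and ray direction set the UV step, with step count blended between Min and Max by view angle.
# A Custom HLSL loop raymarches the heightmap at mip 0, ignoring out-of-bounds samples and interpolating the hit.
# It outputs UVs, offset, mask, depth offset, world position and step-cost color, with low-quality and mobile fallbacks.
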MaterialFunction /Engine/Functions/Engine_MaterialFunctions01/Texturing/ParallaxOcclusionMapping_BoundedUVs
graph {
 properties {
  bExposeToLibrary=true
  description=Parallax Occlusion Mapping uses Ray Tracing to find the correct offsets at each point. Keep in mind that the material cost will not be easily trackable using the instruction count or shader complexity view due to the nature of the loops involved. Use the debug emissive colors to get an idea of expense based on your specified Min and Max Step sizes. This version allows inputting a Bounded UV range and will ignore samples outside and provide a mask for when the trace leaves the bounded region.
  libraryCategoriesText [Misc,Texturing]
 }
 nodes [
  FunctionInput() {
   bUsePreviewValueAsDefault=true
   description=Blank spacer input does nothing
   inputName=-----------------------------
   inputType=FunctionInput_StaticBool
   sortPriority=10
  }
  n1 = QualitySwitch(Default=NamedRerouteUsage(), Low=Constant())
  n2 = Constant()
  FunctionOutput(None=FeatureLevelSwitch(Default=@n1, ES2=@n2, ES3_1=@n2, SM4_REMOVED=@n1)) {description=The offsets only from the parallax., outputName=Intersection Mask, sortPriority=1}
  n3 = NamedRerouteUsage()
  n4 = NamedRerouteUsage()
  n5 = DotProduct(
   A=MaterialFunctionCall() {materialFunction=&CameraDirectionVector}.Result
   B=NamedRerouteUsage()
  )
  n6 = QualitySwitch(Default=Multiply(A=Multiply(A=@n3, B=@n4), B=Abs(None=@n5)), Low=Constant())
  n7 = Constant()
  FunctionOutput(None=FeatureLevelSwitch(Default=@n6, ES2=@n7, ES3_1=@n7, SM4_REMOVED=@n6)) {
   description=Pixel Depth Offset is used to give accurate intersection with other meshes as well as shadows cast by other meshes. Will cause a problem with Dynamic Shadows since the original depths will shadow the new offset depths.
   outputName=Pixel Depth Offset
   sortPriority=4
  }
  MaterialExpressionFloor_12 = Floor(None=Multiply(A=NamedRerouteUsage(), B=NamedRerouteUsage()))
  MaterialExpressionScalarParameter_6 = ScalarParameter() {defaultValue=32, parameterName=Max Expensive Steps Debug Color}
  n8 = Custom!(Steps=@MaterialExpressionFloor_12, MaxSteps=@MaterialExpressionScalarParameter_6) {
   code="float3 colors[10];\r\n\r\ncolors[0]=float3(0,1,0.127);\r\ncolors[1]=float3(0,1,0);\r\ncolors[2]=float3(0.046,0.52,0);\r\ncolors[3]=float3(0.215,0.215,0);\r\ncolors[4]=float3(0.52,0.046,0);\r\ncolors[5]=float3(0.7,0,0);\r\ncolors[6]=float3(1,0,0);\r\ncolors[7]=float3(1,0,0.5);\r\ncolors[8]=float3(1,0.9,0.9);\r\ncolors[9]=float3(1,1,1);\r\n\r\nSteps/=MaxSteps-1;\r\nSteps=saturate(Steps);\r\n\r\nfloat cindex = floor((Steps)*8);\r\nfloat cphase = frac((Steps)*8);\r\n\r\nreturn lerp(colors[cindex],colors[cindex+1],cphase);"
   description=ShaderComplexity
   inputs [
    {
     inputName=Steps
     input {expression=&self:MaterialExpressionFloor_12}
    }
    {
     inputName=MaxSteps
     input {expression=&self:MaterialExpressionScalarParameter_6}
    }
   ]
  }
  FunctionOutput(None=@n8) {
   description=Pixel Depth Offset is used to give accurate intersection with other meshes as well as shadows cast by other meshes. Will cause a problem with Dynamic Shadows since the original depths will shadow the new offset depths.
   outputName=Material Complexity - Steps Debug
   sortPriority=6
  }
  n9 = Multiply(A=NamedRerouteUsage(), B=NamedRerouteUsage())
  n10 = NamedRerouteUsage()
  FunctionOutput(None=Subtract(A=WorldPosition().XYZ, B=Multiply(A=@n9, B=@n10))) {
   description="World Position accounting for offset. Can be used to map virtual worldposition textures by subtracting this Z value as the vertex normal (or 0,0,1 for a floor)."
   outputName=World Position
   sortPriority=5
  }
  n11 = QualitySwitch(Default=Add(A=NamedRerouteUsage(), B=NamedRerouteUsage()), Low=NamedRerouteUsage())
  n12 = NamedRerouteUsage()
  FunctionOutput(None=FeatureLevelSwitch(Default=@n11, ES2=@n12, ES3_1=@n12, SM4_REMOVED=@n11)) {description="This returns the input UVs with Parallax Offset. ", outputName=Parallax UVs}
  n13 = QualitySwitch(Default=NamedRerouteUsage(), Low=Constant())
  n14 = Constant()
  FunctionOutput(None=FeatureLevelSwitch(Default=@n13, ES2=@n14, ES3_1=@n14, SM4_REMOVED=@n13)) {description=The offsets only from the parallax., outputName=Offset Only, sortPriority=1}
  MaterialFunctionCall!(
   Heightmap Texture
   Height Ratio
   Min Steps
   Max Steps
   UVs
   Heightmap Channel
   Reference Plane
   -----------------------------
   Use World Coordinates
   Specify Manual Texture Size
   Manual Texture Size
   ------------------------------
   "Render Shadows (Occlusion Mapping)"
   Light Vector
   Shadow Steps
   Shadow Penumbra
   Transform To VertexNormal
  ) {materialFunction=&/ParallaxOcclusionMapping}
  n15 = FunctionInput() {
   bUsePreviewValueAsDefault=true
   description=The Max Steps is how many steps will be performed at glancing angles where more steps are typically necessary.
   inputName=Max Steps
   inputType=FunctionInput_Scalar
   sortPriority=3
   previewValue=<32 0 0 1>
  }
  n16 = FunctionInput() {
   bUsePreviewValueAsDefault=true
   description=The Min steps is how many steps will be used when looking downward at the material. Typically fewer steps are required when looking down since there is less resulting offset.
   inputName=Min Steps
   inputType=FunctionInput_Scalar
   sortPriority=2
   previewValue=<8 0 0 1>
  }
  n17 = DotProduct(A=NamedRerouteUsage(), B=VertexNormalWS())
  n18 = LinearInterpolate(A=@n15, B=@n16, Alpha=Saturate(None=Abs(None=@n17)))
  NamedRerouteDeclaration(None=Floor(None=@n18)) {name=Num Steps, nodeColor=<rgba 0 1 0.7656 1>}
  NamedRerouteDeclaration(None=Divide(B=@n18)) {name=Step Size, nodeColor=<rgba 0 1 0.9063 1>}
  n19 = FunctionInput() {
   bUsePreviewValueAsDefault=true
   description=Only used for Pixel Depth Offset. This lets you manually specify the World Size of your texture to skip some expensive instructions that are normally used to determine the UV to World Size.
   inputName=Manual Texture Size
   inputType=FunctionInput_Scalar
   sortPriority=13
   previewValue=<1024 0 0 1>
  }
  n20 = NamedRerouteUsage()
  n21 = Abs(None=DDX(Value=@n20))
  n22 = WorldPosition()
  n23 = DDX(Value=@n22.XYZ)
  n24 = Abs(None=DDY(Value=@n20))
  n25 = DDY(Value=@n22.XYZ)
  n26 = Max(
   A=Divide(
    A=MaterialFunctionCall(Vector 2=@n21) {materialFunction=/Engine/Functions/Engine_MaterialFunctions02/Utility/VectorLength}.V2 Length
    B=MaterialFunctionCall(Vector 3=@n23) {materialFunction=&VectorLength}.V3 Length
   )
   B=Divide(
    A=MaterialFunctionCall(Vector 2=@n24) {materialFunction=&VectorLength}.V2 Length
    B=MaterialFunctionCall(Vector 3=@n25) {materialFunction=&VectorLength}.V3 Length
   )
  )
  n27 = FunctionInput(Preview=StaticBool()) {
   bUsePreviewValueAsDefault=true
   description=Only used for Pixel Depth Offset. This lets you manually specify the World Size of your texture to skip some expensive instructions that are normally used to determine the UV to World Size.Defaults to False.
   inputName=Specify Manual Texture Size
   inputType=FunctionInput_StaticBool
   sortPriority=12
  }
  NamedRerouteDeclaration(None=StaticSwitch(True=@n19, False=Divide(B=@n26), Value=@n27)) {name=World Size, nodeColor=<rgba 0.7578 1 0 1>}
  MaterialExpressionFunctionInput_9 = FunctionInput(Preview=TextureObject() {texture=/Engine/EngineVolumetrics/Fogsheet/Materials/T_EV_BlankWhite_01}) {description=The Texture Object containing your Heightmap, inputName=Heightmap Texture, inputType=FunctionInput_Texture2D}
  MaterialExpressionNamedRerouteUsage_5 = NamedRerouteUsage()
  MaterialExpressionNamedRerouteUsage_8 = NamedRerouteUsage()
  MaterialExpressionNamedRerouteUsage_9 = NamedRerouteUsage()
  MaterialExpressionNamedRerouteUsage_13 = NamedRerouteUsage()
  n28 = NamedRerouteUsage()
  MaterialExpressionDDX_4 = DDX(Value=@n28)
  MaterialExpressionDDY_4 = DDY(Value=@n28)
  MaterialExpressionFunctionInput_3 = FunctionInput() {
   bUsePreviewValueAsDefault=true
   description=Which Channel your heightmap texture uses. Uses Alpha channel by Default.
   inputName=Heightmap Channel
   inputType=FunctionInput_Vector4
   sortPriority=8
  }
  MaterialExpressionFunctionInput_6 = FunctionInput() {
   bUsePreviewValueAsDefault=true
   description=Which Channel your heightmap texture uses. Uses Alpha channel by Default.
   inputName=ClampMinMax
   inputType=FunctionInput_Vector4
   sortPriority=16
   previewValue=<0 0 1 1>
  }
  n29 = Custom!(
   Tex=@MaterialExpressionFunctionInput_9
   UV=@MaterialExpressionNamedRerouteUsage_5
   MaxSteps=@MaterialExpressionNamedRerouteUsage_8
   stepsize=@MaterialExpressionNamedRerouteUsage_9
   UVDist=@MaterialExpressionNamedRerouteUsage_13
   InDDX=@MaterialExpressionDDX_4
   InDDY=@MaterialExpressionDDY_4
   HeightMapChannel=@MaterialExpressionFunctionInput_3
   MinMax=@MaterialExpressionFunctionInput_6
  ) {
   code="float rayheight=1;\r\nfloat oldray=1;\r\nfloat2 offset=0;\r\nfloat oldtex=1;\r\nfloat texatray;\r\nfloat yintersect;\r\nint i=0;\r\n\r\n\r\nwhile (i<MaxSteps+2)\r\n{\r\n\tfloat2 curUV = UV+offset;\r\n\t//curUV = clamp(curUV, MinMax.rg, MinMax.ba);\r\n\r\n\t\r\n\ttexatray=dot(HeightMapChannel, Tex.SampleLevel(TexSampler, curUV, 0));\r\n\t\r\n\tif(curUV.x < MinMax.r || curUV.y < MinMax.g || curUV.x > MinMax.b || curUV.y > MinMax.a)\r\n\t{\r\n\t\ttexatray = -1000000;\r\n\t}\r\n\t\r\n\tif (rayheight < texatray)\r\n\t{\r\n\t\tfloat xintersect = (oldray-oldtex)+(texatray-rayheight);\r\n\t\txintersect=(texatray-rayheight)/xintersect;\r\n\t\tyintersect=(oldray*(xintersect))+(rayheight*(1-xintersect));\r\n\t\toffset-=(xintersect*UVDist);\r\n\r\nfloat4 output;\r\noutput.xy=offset;\r\noutput.z=yintersect;\r\noutput.a=1;\r\nreturn output;\r\n\r\n\t}\r\n\t\r\n\toldray=rayheight;\r\n\trayheight-=stepsize;\r\n\toffset+=UVDist;\r\n\toldtex=texatray;\r\n\t\r\n\t\r\n\ti++;\r\n}\r\n\r\nreturn 0;"
   description=Parallax Only Mapping
   outputType=CMOT_Float4
   inputs [
    {
     inputName=Tex
     input {expression=&self:MaterialExpressionFunctionInput_9}
    }
    {
     inputName=UV
     input {expression=&self:MaterialExpressionNamedRerouteUsage_5}
    }
    {
     inputName=MaxSteps
     input {expression=&self:MaterialExpressionNamedRerouteUsage_8}
    }
    {
     inputName=stepsize
     input {expression=&self:MaterialExpressionNamedRerouteUsage_9}
    }
    {
     inputName=UVDist
     input {expression=&self:MaterialExpressionNamedRerouteUsage_13}
    }
    {
     inputName=InDDX
     input {expression=&self:MaterialExpressionDDX_4}
    }
    {
     inputName=InDDY
     input {expression=&self:MaterialExpressionDDY_4}
    }
    {
     inputName=HeightMapChannel
     input {expression=&self:MaterialExpressionFunctionInput_3}
    }
    {
     inputName=MinMax
     input {expression=&self:MaterialExpressionFunctionInput_6}
    }
   ]
  }
  NamedRerouteDeclaration(None=ComponentMask(None=@n29) {a=true}) {name=Mask, nodeColor=<rgba 0 0.1562 1 1>}
  NamedRerouteDeclaration(None=FunctionInput(Preview=CameraVectorWS()) {bUsePreviewValueAsDefault=true, description=Blank spacer input does nothing, inputName=Ray Direction, sortPriority=20}) {name=Ray Direction, nodeColor=<rgba 0 0.3203 1 1>}
  n30 = NamedRerouteUsage()
  n31 = VertexNormalWS()
  n32 = FunctionInput(Preview=StaticBool() {value=true}) {
   bUsePreviewValueAsDefault=true
   inputName=Transform To VertexNormal
   inputType=FunctionInput_StaticBool
   sortPriority=50
   previewValue=<1 1 1 1>
  }
  NamedRerouteDeclaration(None=StaticSwitch(
    True=MaterialFunctionCall!(Vector to Transform=@n30, Z Vector=@n31, Center Location) {materialFunction=/Engine/Functions/Engine_MaterialFunctions02/Math/TransformToZVector}.Result
    False=@n30
    Value=@n32
   )) {name=RayDirectionWS, nodeColor=<rgba 0.1016 1 0 1>}
  n33 = FunctionInput(Preview=TextureCoordinate()) {
   bUsePreviewValueAsDefault=true
   description=The coordinates to use for mapping. Can be used with worldposition V2 values as well.
   inputName=UVs
   inputType=FunctionInput_Vector2
   sortPriority=4
  }
  n34 = NamedRerouteUsage()
  n35 = FunctionInput() {
   bUsePreviewValueAsDefault=true
   description="The relative point from which the parallax will be applied. When 1 or higher, all parallax will be downwards. When 0, all parallax will be upwards."
   inputName=Reference Plane
   inputType=FunctionInput_Scalar
   sortPriority=9
   previewValue=<1 0 0 1>
  }
  n36 = Multiply(A=OneMinus(None=@n35)) {constB=-1}
  NamedRerouteDeclaration(None=Add(A=@n33, B=Multiply(A=@n34, B=@n36))) {name=UVs, nodeColor=<rgba 1 0 0.1641 1>}
  n37 = FunctionInput() {
   bUsePreviewValueAsDefault=true
   description=This is how 'deep' the heightmap is in relation to its width. Typical values are between 0.05 and 0.1.
   inputName=Height Ratio
   inputType=FunctionInput_Scalar
   sortPriority=1
   previewValue=<0.075 0 0 1>
  }
  n38 = Multiply(A=NamedRerouteDeclaration(None=@n37) {name=Height Ratio, nodeColor=<rgba 0.8984 1 0 1>}, B=NamedRerouteUsage())
  NamedRerouteDeclaration(None=Multiply(A=NamedRerouteDeclaration(None=@n38) {name=Ray XY Scaled, nodeColor=<rgba 1 0 0.2344 1>}, B=NamedRerouteUsage()) {desc=smooth}) {name=RayStep, nodeColor=<rgba 0.1484 0 1 1>}
  n39 = FunctionInput(Preview=StaticBool()) {
   bUsePreviewValueAsDefault=true
   description="If using world coordinates as the UVs, the shader needs to skip the tangent->world transform. This saves a few instructions. Default = FALSE"
   inputName=Use World Coordinates
   inputType=FunctionInput_StaticBool
   sortPriority=11
  }
  n40 = StaticSwitch(
   True=NamedRerouteUsage()
   False=Transform(None=NamedRerouteUsage()) {transformSourceType=TRANSFORMSOURCE_World, transformType=TRANSFORM_Tangent}
   Value=@n39
  )
  n41 = ComponentMask(None=@n40) {g=true, r=true}
  NamedRerouteDeclaration(None=Divide(A=Multiply(A=@n41) {constB=-1}, B=ComponentMask(None=@n40) {b=true}) {constB=1}) {name=Ray XY, nodeColor=<rgba 1 0 0.375 1>}
  n42 = NamedRerouteDeclaration(None=ComponentMask(None=@n29) {g=true, r=true}) {name=Parallax, nodeColor=<rgba 0.5703 1 0 1>}
  n43 = ComponentMask(None=@n29) {b=true}
  n44 = Multiply(A=NamedRerouteDeclaration(None=OneMinus(None=@n43)) {name=Surface Height, nodeColor=<rgba 0 0.8828 1 1>}, B=NamedRerouteUsage())
  NamedRerouteDeclaration(None=Distance(A=AppendVector(A=@n42, B=@n44), B=Constant())) {name=Distance to Surface, nodeColor=<rgba 0.2891 0 1 1>}
 ]
}
