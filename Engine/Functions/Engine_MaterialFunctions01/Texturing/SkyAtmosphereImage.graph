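# Material function that projects a texture as an oriented sprite on the sky along a given direction, with scale and rotation.
# A Custom HLSL node builds a tangent basis around the normalized direction, projects the camera vector onto it, then scales and rotates it into UVs.
# It outputs unmasked RGB and alpha plus a separate Mask that excludes the opposite hemisphere and out-of-range UVs.
# Rotation sine comes from a square root, so is never negative.
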MaterialFunction /Engine/Functions/Engine_MaterialFunctions01/Texturing/SkyAtmosphereImage
graph {
 properties {bExposeToLibrary=true, libraryCategoriesText [Misc,Texturing]}
 nodes [
  n1 = FunctionInput() {
   bUsePreviewValueAsDefault=true
   description=The direction in which the sky sprite needs to appear.
   inputName=Direction
   sortPriority=1
   previewValue=<1 0 1 1>
  }
  MaterialExpressionNormalize_3 = Normalize(VectorInput=@n1)
  MaterialExpressionCameraVectorWS_0 = CameraVectorWS()
  MaterialExpressionFunctionInput_3 = FunctionInput() {bUsePreviewValueAsDefault=true, inputName=Scale, inputType=FunctionInput_Vector2, sortPriority=2, previewValue=<1 1 0 1>}
  MaterialExpressionFunctionInput_0 = FunctionInput() {bUsePreviewValueAsDefault=true, inputName="Rotation (radian)", inputType=FunctionInput_Scalar, sortPriority=3, previewValue=<-0.25 0 0 1>}
  n2 = Custom!(
   LightDirection=@MaterialExpressionNormalize_3
   ViewDirection=@MaterialExpressionCameraVectorWS_0
   Size=@MaterialExpressionFunctionInput_3
   Rotation=@MaterialExpressionFunctionInput_0
  ) {
   code="float3 Up = float3(0.0, 0.0, 1.0);\r\nfloat3 Right = normalize(cross(Up, LightDirection.xyz));\r\nUp = normalize(cross(LightDirection.xyz, Right));\r\n\r\n//float3 Right = float3(-1.0, 0.0, 0.0);\r\n//float3 Up = normalize(cross(LightDirection.xyz, Right));\r\n//Right = normalize(cross(Up, LightDirection.xyz));\r\n\r\nfloat3 UVW = float3(dot(ViewDirection.xyz,Right), dot(ViewDirection.xyz,Up), -dot(ViewDirection.xyz,LightDirection.xyz));\r\nfloat2 UV = UVW.xy;\r\n\r\nMask = UVW.z > 0.0 ? 1.0f : 0.0f;\r\n\r\nUV = UV/Size;\r\n\r\nfloat CosAngle = cos(Rotation.x);\r\nfloat SinAngle = sqrt(max(0.0, 1.0-CosAngle*CosAngle));\r\nfloat2 Rot0 =-float2(CosAngle, SinAngle);\r\nfloat2 Rot1 =-float2(-Rot0.y, Rot0.x);\r\nUV = float2(dot(UV, Rot0), dot(UV, Rot1));\r\nUV += 0.5;\r\n\r\nMask *= any(UV.xy>1.0) || any(UV.xy<0.0) ? 0.0 : 1.0;\r\n\r\nreturn UV;"
   description=SkyAtmosphereImage
   outputType=CMOT_Float2
   additionalOutputs [{outputName=Mask, outputType=CMOT_Float1}]
   inputs [
    {
     inputName=LightDirection
     input {expression=&self:MaterialExpressionNormalize_3}
    }
    {
     inputName=ViewDirection
     input {expression=&self:MaterialExpressionCameraVectorWS_0}
    }
    {
     inputName=Size
     input {expression=&self:MaterialExpressionFunctionInput_3}
    }
    {
     inputName=Rotation
     input {expression=&self:MaterialExpressionFunctionInput_0}
    }
   ]
  }
  n3 = TextureSample(
   UVs=@n2.return
   Tex=FunctionInput(Preview=TextureObject() {texture=/Engine/Functions/Engine_MaterialFunctions02/ExampleContent/Textures/flipbook}) {description=The texture applied on the sky sprite., inputName=TextureObject, inputType=FunctionInput_Texture2D}
  ) {texture=&flipbook}
  FunctionOutput(None=@n3.RGB) {description="UnMasked RGB image, may contain stretching, use output mask "}
  FunctionOutput(None=@n2.Mask) {description=Mask for the area of correct projection, outputName=Mask, sortPriority=2}
  FunctionOutput(None=@n3.A) {description="UnMasked image alpha, may contain stretching, use output mask ", outputName=Alpha, sortPriority=1}
 ]
}
