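# Material function that projects a texture onto an infinite sphere along an ImageVector direction, for placing images in a skybox.
# RotateVector turns the camera vector relative to that direction; its XY becomes UVs, rotated by a 0-1 input and center-scaled using only Scale's G, with V flipped.
# It samples unmasked RGB and Alpha, plus a Mask from ceiled U bounds and a negated-Z facing test that excludes the opposite-side projection.
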MaterialFunction /Engine/Functions/Engine_MaterialFunctions01/Texturing/SkyboxImage
graph {
 properties {
  bExposeToLibrary=true
  description="Projects an image onto an infinite sphere, useful for putting images in the skybox"
  libraryCategoriesText [Misc,Texturing]
 }
 nodes [
  n1 = FunctionInput() {bUsePreviewValueAsDefault=true, inputName=ImageVector, sortPriority=1, previewValue=<1 0 1 1>}
  n2 = MaterialFunctionCall!(VectorToRotate=CameraVectorWS(), LookAtVector=Normalize(VectorInput=@n1), RestingVector) {materialFunction=/Engine/Functions/Engine_MaterialFunctions01/Vectors/RotateVector}
  n3 = ComponentMask(None=@n2.Result) {g=true, r=true}
  n4 = FunctionInput() {bUsePreviewValueAsDefault=true, inputName=Rotation 0-1, inputType=FunctionInput_Scalar, sortPriority=3, previewValue=<-0.25 0 0 1>}
  n5 = FunctionInput() {bUsePreviewValueAsDefault=true, inputName=Scale, inputType=FunctionInput_Vector2, sortPriority=2, previewValue=<1 1 0 1>}
  n6 = MaterialFunctionCall(Float2=@n5) {materialFunction=&BreakOutFloat2Components}
  n7 = MaterialFunctionCall(
   UVs=Rotator(Coordinate=Add(A=@n3) {constB=0.5}, Time=Multiply(A=@n4) {constB=6.28}) {speed=1}
   Texture Scale=Multiply(A=AppendVector(A=@n6.G, B=@n6.G), B=Constant2Vector() {g=-1, r=1})
  ) {materialFunction=&ScaleUVsByCenter}
  n8 = MaterialFunctionCall(Float2=Clamp(None=Ceil(None=@n7.UVs))) {materialFunction=&BreakOutFloat2Components}
  n9 = Min(A=@n8.R, B=@n8.R)
  n10 = OneMinus(None=@n7.UVs)
  n11 = MaterialFunctionCall(Float2=Clamp(None=Ceil(None=@n10))) {materialFunction=&BreakOutFloat2Components}
  n12 = Min(A=@n11.R, B=@n11.R)
  n13 = Clamp(None=Multiply(A=ComponentMask(None=@n2.Result) {b=true}) {constB=-1})
  FunctionOutput(None=Multiply(A=Min(A=@n9, B=@n12), B=Ceil(None=@n13))) {description=Mask for the area of correct projection, outputName=Mask, sortPriority=2}
  n14 = TextureSample(
   UVs=@n7.UVs
   Tex=FunctionInput(Preview=TextureObject() {texture=/Engine/Functions/Engine_MaterialFunctions02/ExampleContent/Textures/flipbook}) {inputName=TextureObject, inputType=FunctionInput_Texture2D}
  ) {texture=&flipbook}
  FunctionOutput(None=@n14.RGB) {description="UnMasked RGB image, may contain stretching, use output mask ", outputName=RGB}
  FunctionOutput(None=@n14.A) {description="UnMasked image alpha, may contain stretching, use output mask ", outputName=Alpha, sortPriority=1}
 ]
}
