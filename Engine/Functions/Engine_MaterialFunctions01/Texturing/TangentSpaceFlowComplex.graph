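# Material function that animates a flow texture along a tangent-space FlowDirection, scaled by time, speed and strength inputs.
# Two looping Frac phases, half a cycle apart, offset the UVs opposite the direction vector for two mip-biased samples of the same texture.
# A smoothed triangle-wave alpha crossfades them, hiding each phase reset.
# Outputs are the blended result, both UV sets and the lerp alpha.
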MaterialFunction /Engine/Functions/Engine_MaterialFunctions01/Texturing/TangentSpaceFlowComplex
graph {
 properties {description=Pushing the FlowTexture along the direction of the FlowDirection, libraryCategoriesText [Misc,Texturing]}
 nodes [
  n1 = ComponentMask(None=FunctionInput(Preview=TextureCoordinate()) {bUsePreviewValueAsDefault=true, inputName=UVs, inputType=FunctionInput_Vector2}) {g=true, r=true}
  n2 = FunctionInput(Preview=ComponentMask(None=Constant3Vector() {constant=<rgba 0 1 0 0>}) {g=true, r=true}) {
   bUsePreviewValueAsDefault=true
   description=Flow Direction in Tangent Space from -1 to 1
   inputName=FlowDirection
   inputType=FunctionInput_Vector2
  }
  n3 = Multiply(A=@n2, B=Constant2Vector() {g=-1, r=-1})
  n4 = FunctionInput(Preview=Time()) {bUsePreviewValueAsDefault=true, inputName=Time, inputType=FunctionInput_Scalar, sortPriority=10, previewValue=<0.2 0 0 1>}
  n5 = FunctionInput() {bUsePreviewValueAsDefault=true, inputName=FlowSpeed, inputType=FunctionInput_Scalar, previewValue=<0.2 0 0 1>}
  n6 = Constant() {desc=Cycle Variable, r=1}
  n7 = Multiply(A=Multiply(A=@n4, B=@n5), B=@n6)
  n8 = Frac(None=@n7)
  n9 = FunctionInput() {bUsePreviewValueAsDefault=true, inputName=FlowStrength, inputType=FunctionInput_Scalar, previewValue=<0.2 0 0 1>}
  n10 = Multiply(A=Constant() {r=1}, B=@n9)
  n11 = Add(A=@n1, B=Multiply(A=@n3, B=Multiply(A=@n8, B=@n10)))
  n12 = FunctionInput(Preview=TextureObject() {texture=/Engine/EngineMaterials/DefaultDiffuse}) {inputName=FlowTexture, inputType=FunctionInput_Texture2D}
  n13 = FunctionInput() {bUsePreviewValueAsDefault=true, inputName=MipBias, inputType=FunctionInput_Scalar, sortPriority=20}
  n14 = TextureSample!(UVs=@n11, Tex=@n12, Bias=@n13, Apply View MipBias) {mipValueMode=TMVM_MipBias}
  n15 = Multiply(A=@n6, B=Constant() {r=0.5})
  n16 = Frac(None=Add(A=@n7, B=@n15))
  n17 = Add(A=@n1, B=Multiply(A=@n3, B=Multiply(A=@n10, B=@n16)))
  n18 = TextureSample(UVs=@n17, Tex=@n12, Bias=@n13) {mipValueMode=TMVM_MipBias}
  n19 = Divide(A=Abs(None=Subtract(A=@n8, B=@n15)), B=@n15)
  n20 = OneMinus(None=Multiply(A=@n19, B=@n19))
  n21 = OneMinus(None=Multiply(A=@n20, B=@n20))
  FunctionOutput(None=LinearInterpolate(A=AppendVector(A=@n14.RGB, B=@n14.RGB), B=AppendVector(A=@n18.RGB, B=@n18.RGB), Alpha=@n21))
  FunctionOutput(None=@n11) {outputName=Uv1, sortPriority=1}
  FunctionOutput(None=@n17) {outputName=Uv2, sortPriority=2}
  FunctionOutput(None=@n21) {outputName=LerpAlpha, sortPriority=3}
 ]
}
