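# Material function that takes one texture-bomb sample with parallax occlusion mapping, outputting the randomized pre-parallax UVs and the sampled color.
# A pseudo-random hash of the cell index offsets, mirrors and axis-swaps the UVs.
# The tangent-space camera vector gets the same flip before Parallax_For_Bomb traces the heightmap with temporally dithered max steps.
# Sampling uses explicit, flipped derivatives from the original UVs, and an IsNormal switch re-orients normal-map results to match.
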MaterialFunction /Engine/Functions/Engine_MaterialFunctions01/Texturing/TextureBomb_SingleSample_POM
graph {
 properties {libraryCategoriesText []}
 nodes [
  MaterialExpressionFunctionInput_15 = FunctionInput() {bUsePreviewValueAsDefault=true, inputName=Cell Index, inputType=FunctionInput_Vector2}
  n1 = Custom!(pos=@MaterialExpressionFunctionInput_15) {
   code="float2 noise = dot(pos.xy * pos.xy, float2(12.581057, 11.129465));\r\nnoise.y = dot(pos.xy * pos.xy/5, float2(46.538326, 11.057759));\r\n\r\nreturn frac(noise);"
   description=Pseudo Random
   outputType=CMOT_Float2
   inputs [{
     inputName=pos
     input {expression=&self:MaterialExpressionFunctionInput_15}
    }]
  }
  n2 = FunctionInput() {bUsePreviewValueAsDefault=true, inputName=Offset Strength, inputType=FunctionInput_Scalar, sortPriority=2, previewValue=<1 0 0 1>}
  n3 = Clamp(None=Ceil(None=Subtract(A=@n1) {constB=0.5}))
  n4 = LinearInterpolate(A=Constant2Vector() {g=1, r=1}, B=Constant2Vector() {g=-1, r=1}, Alpha=ComponentMask(None=@n3) {g=true})
  n5 = FunctionInput(Preview=TextureCoordinate()) {bUsePreviewValueAsDefault=true, inputName=UVs, inputType=FunctionInput_Vector2, sortPriority=1}
  n6 = Multiply(A=@n4, B=@n5)
  n7 = ComponentMask(None=@n3) {r=true}
  n8 = Add(
   A=Multiply(A=@n1, B=@n2)
   B=LinearInterpolate(A=MaterialFunctionCall(XY=@n6) {materialFunction=&Swizzle}.YX, B=@n6, Alpha=@n7) {desc=Flipped UVs}
  )
  FunctionOutput(None=@n8) {outputName=UVs}
  n9 = LinearInterpolate(A=MaterialFunctionCall(XY=@n4) {materialFunction=&Swizzle}.YX, B=@n4, Alpha=@n7) {desc=Flip Term - Normal Flipping}
  n10 = FunctionInput() {bUsePreviewValueAsDefault=true, inputName=HeightTex, inputType=FunctionInput_Texture2D, sortPriority=6, previewValue=<1 0 0 1>}
  n11 = FunctionInput() {bUsePreviewValueAsDefault=true, inputName=HeightRatio, inputType=FunctionInput_Scalar, sortPriority=7, previewValue=<1 0 0 1>}
  n12 = LinearInterpolate(Alpha=MaterialFunctionCall() {materialFunction=&DitherTemporalAA}.Result) {constA=1, constB=1.2}
  n13 = Multiply(
   A=AppendVector(A=@n4, B=Constant() {r=1})
   B=Transform(None=CameraVectorWS()) {transformSourceType=TRANSFORMSOURCE_World, transformType=TRANSFORM_Tangent}
  )
  n14 = MaterialFunctionCall(XYZ=@n13) {materialFunction=&Swizzle}
  n15 = FunctionInput() {bUsePreviewValueAsDefault=true, inputName=TexObject, inputType=FunctionInput_Texture2D, sortPriority=3, previewValue=<1 0 0 1>}
  n16 = Multiply(A=@n9)
  n17 = Multiply(A=DDX(Value=@n5), B=@n16)
  n18 = ComponentMask(None=@n7) {r=true}
  n19 = Multiply(A=DDY(Value=@n5), B=@n16)
  n20 = TextureSample!(
   UVs=MaterialFunctionCall!(
    Heightmap Texture=@n10
    Height Ratio=@n11
    Min Steps
    Max Steps=Multiply(A=@n12) {constB=64}
    UVs=@n8
    Heightmap Channel=Constant4Vector() {constant=<rgba 1 0 0 0>}.RGBA
    Reference Plane
    -----------------------------
    Specify Manual Texture Size
    Manual Texture Size
    Override Camera Vector=LinearInterpolate(A=@n14.YXZ, B=@n13, Alpha=@n7)
    ------------------------------
    "Render Shadows (Occlusion Mapping)"
    Light Vector
    Shadow Steps
    Shadow Penumbra
   ) {materialFunction=&/Parallax_For_Bomb}.Parallax UVs
   Tex=@n15
   "DDX(UVs)"=LinearInterpolate(A=MaterialFunctionCall(XY=@n17) {materialFunction=&Swizzle}.YX, B=@n17, Alpha=@n18)
   "DDY(UVs)"=LinearInterpolate(A=MaterialFunctionCall(XY=@n19) {materialFunction=&Swizzle}.YX, B=@n19, Alpha=@n18)
   Apply View MipBias
  ) {desc=Top Right, mipValueMode=TMVM_Derivative}
  n21 = Multiply(A=AppendVector(A=@n9, B=Constant() {r=1}), B=@n20.RGB)
  n22 = MaterialFunctionCall(XYZ=@n21) {materialFunction=&Swizzle}
  n23 = StaticBool()
  FunctionOutput(None=StaticSwitch(
    True=LinearInterpolate(A=@n22.YXZ, B=@n21, Alpha=@n7)
    False=@n20.RGB
    Value=FunctionInput(Preview=@n23) {bUsePreviewValueAsDefault=true, inputName=IsNormal?, inputType=FunctionInput_StaticBool, sortPriority=4}
   )) {outputName=Texture Sample, sortPriority=1}
  n24 = LinearInterpolate(A=Constant2Vector() {g=1, r=1}, B=Constant2Vector() {g=1, r=-1}, Alpha=@n7)
  Multiply(A=@n6, B=@n24)
  Multiply(A=@n4, B=@n24)
 ]
}
