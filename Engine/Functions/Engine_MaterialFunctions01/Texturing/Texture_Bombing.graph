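# Material function exposed to the library under Texturing that breaks up tiling artifacts through texture bombing, blending randomly offset texture samples.
# Tiled UVs define grid cells, and four neighboring cells each sample the texture through TextureBomb_SingleSample with an offset strength and normal-map switch.
# The samples blend bilinearly on fractional UVs, with switchable cosine smoothing and optional HeightLerp sharpening from a heightmap.
# Several leftover nodes stay unconnected to the output.
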MaterialFunction /Engine/Functions/Engine_MaterialFunctions01/Texturing/Texture_Bombing
graph {
 properties {
  bExposeToLibrary=true
  description=Texture bombing uses multiple offset texture samples to break up tiling artifacts.
  libraryCategoriesText [Texturing]
 }
 nodes [
  Constant3Vector() {desc=Test colors, constant=<rgba 0.2 1 0.2 0>}
  Constant3Vector() {desc=Test colors, constant=<rgba 0.2 0.2 1 0>}
  Constant3Vector() {desc=Test colors, constant=<rgba 1 0.2 0.2 0>}
  n1 = Custom!(pos, Rand) {
   code="float2 noise = dot(pos.xy * pos.xy, float2(12.581057, 11.129465));\r\nnoise.y = dot(pos.xy * pos.xy/5, float2(46.538326, 11.057759));\r\n\r\nreturn frac(noise);"
   description=Pseudo Random
   outputType=CMOT_Float2
   inputs [{inputName=pos, input {expression=null}}, {inputName=Rand, input {expression=null}}]
  }
  Divide(A=@n1) {constB=-1}
  n2 = FunctionInput() {bUsePreviewValueAsDefault=true, inputName=Tiling, inputType=FunctionInput_Scalar, sortPriority=2, previewValue=<4 0 0 1>}
  n3 = Multiply(
   A=FunctionInput(Preview=TextureCoordinate()) {bUsePreviewValueAsDefault=true, inputName=UVs, inputType=FunctionInput_Vector2, sortPriority=1}
   B=@n2
  ) {constB=4}
  n4 = VectorParameter() {parameterName=t}
  n5 = Ceil(None=Add(A=@n3, B=ComponentMask(None=@n4.RGB) {g=true, r=true}))
  n6 = Constant2Vector() {g=-1, r=-1}
  n7 = FunctionInput() {
   bUsePreviewValueAsDefault=true
   description="The strength of the random offsets. "
   inputName=Offset
   inputType=FunctionInput_Scalar
   sortPriority=3
   previewValue=<1 0 0 1>
  }
  n8 = Multiply(A=@n7) {constB=-1}
  n9 = FunctionInput(Preview=TextureObject() {texture=/Engine/EditorResources/S_VectorFieldVol}) {bUsePreviewValueAsDefault=true, inputName=Texture Object, inputType=FunctionInput_Texture2D}
  n10 = FunctionInput(Preview=StaticBool()) {
   bUsePreviewValueAsDefault=true
   description=Normal maps require a few extra instructions to account for the randomly flipping directions. Make sure to set this to TRUE when using a normal map as the input texture object.
   inputName=Is Normalmap
   inputType=FunctionInput_StaticBool
   sortPriority=8
   previewValue=<4 0 0 1>
  }
  n11 = Multiply(A=MaterialFunctionCall!(Cell Index=Add(A=@n5, B=@n6) {constB=-1, desc=Top Left}, UVs=Multiply(A=@n3) {desc=UVs}, Offset Strength=@n8, TexObject=@n9, IsNormal?=@n10) {materialFunction=&/TextureBomb_SingleSample}.Texture Sample)
  n12 = MaterialFunctionCall(
   Cell Index=Add(A=@n5, B=Constant2Vector() {g=-1}) {constB=-1, desc=Top Right}
   UVs=Multiply(A=@n3) {desc=UVs}
   Offset Strength=@n8
   TexObject=@n9
   IsNormal?=@n10
  ) {materialFunction=&TextureBomb_SingleSample}
  n13 = Multiply(A=@n12.Texture Sample)
  n14 = Multiply(A=@n3) {desc=UVs}
  MaterialExpressionFrac_0 = Frac(None=@n14)
  n15 = Custom!(uv=@MaterialExpressionFrac_0) {
   code="float2 cosblend = (1-cos(frac(uv)*PI))*0.5;\r\nreturn cosblend;"
   description=Cosine Blend
   outputType=CMOT_Float2
   inputs [{inputName=uv, input {expression=&self:MaterialExpressionFrac_0}}]
  }
  n16 = StaticSwitchParameter(True=@n15, False=@MaterialExpressionFrac_0) {parameterName=Cosine Blend}
  n17 = FunctionInput(Preview=TextureObject() {texture=&S_VectorFieldVol}) {
   bUsePreviewValueAsDefault=true
   description="The heightmap input is used only when \"Enable Height Lerp\" is Enabled. It is used to sharpen the blend between the textures to reduce blurring that occurs with simple blended transitions."
   inputName=Optional Heightmap
   inputType=FunctionInput_Texture2D
   sortPriority=6
  }
  n18 = DDX(Value=@n14)
  n19 = DDY(Value=@n14)
  n20 = TextureSample!(UVs=@n12.UVs, Tex=@n17, "DDX(UVs)"=@n18, "DDY(UVs)"=@n19, Apply View MipBias) {desc=Height, mipValueMode=TMVM_Derivative}
  n21 = AppendVector(A=@n20.RGB, B=@n20.RGB)
  n22 = FunctionInput() {bUsePreviewValueAsDefault=true, inputName=Contrast, inputType=FunctionInput_Scalar, sortPriority=7}
  n23 = MaterialFunctionCall(Float2=@n16) {materialFunction=&BreakOutFloat2Components}
  n24 = FunctionInput(Preview=StaticBool()) {bUsePreviewValueAsDefault=true, inputName=Enable Height Lerp, inputType=FunctionInput_StaticBool, sortPriority=8}
  n25 = StaticSwitch(
   True=MaterialFunctionCall(
    Transition Phase=MaterialFunctionCall(Float2=@n16) {materialFunction=&BreakOutFloat2Components}.R
    Height Texture=StaticComponentMaskParameter(None=@n21) {defaultR=true, parameterName=Heightmap Channel}
    Contrast=@n22
   ) {desc=X bkend bottom row, materialFunction=&HeightLerp}.Alpha
   False=@n23.R
   Value=@n24
  )
  n26 = MaterialFunctionCall(Cell Index=Add(A=@n5, B=Constant2Vector() {r=-1}) {constB=-1, desc=Bottom Left}, UVs=@n3, Offset Strength=@n8, TexObject=@n9, IsNormal?=@n10) {materialFunction=&TextureBomb_SingleSample}
  n27 = Multiply(A=@n26.Texture Sample)
  n28 = MaterialFunctionCall(Cell Index=Add(A=@n5) {constB=0, desc=Current Cell}, UVs=@n14, Offset Strength=@n8, TexObject=@n9, IsNormal?=@n10) {materialFunction=&TextureBomb_SingleSample}
  n29 = Multiply(A=@n28.Texture Sample)
  n30 = MaterialFunctionCall(Float2=@n16) {materialFunction=&BreakOutFloat2Components}
  n31 = TextureSample(UVs=@n28.UVs, Tex=@n17, "DDX(UVs)"=@n18, "DDY(UVs)"=@n19) {desc=Height, mipValueMode=TMVM_Derivative}
  n32 = StaticComponentMaskParameter(None=AppendVector(A=@n31.RGB, B=@n31.RGB)) {defaultR=true, parameterName=Heightmap Channel}
  n33 = MaterialFunctionCall(Transition Phase=@n30.R, Height Texture=@n32, Contrast=@n22) {desc=X blend, materialFunction=&HeightLerp}
  n34 = StaticSwitch(True=@n33.Alpha, False=@n23.R, Value=@n24)
  n35 = MaterialFunctionCall(Float2=@n16) {materialFunction=&BreakOutFloat2Components}
  n36 = TextureSample(UVs=@n26.UVs, Tex=@n17, "DDX(UVs)"=@n18, "DDY(UVs)"=@n19) {desc=Height, mipValueMode=TMVM_Derivative}
  n37 = MaterialFunctionCall(
   Transition Phase=@n35.G
   Height Texture=StaticComponentMaskParameter(None=AppendVector(A=@n36.RGB, B=@n36.RGB)) {defaultR=true, parameterName=Heightmap Channel}
   Contrast=@n22
  ) {desc=Y Blend for Left side, materialFunction=&HeightLerp}
  n38 = MaterialFunctionCall(Transition Phase=@n30.G, Height Texture=@n32, Contrast=@n22) {desc=Y Blend, materialFunction=&HeightLerp}
  n39 = LinearInterpolate(A=@n37.Alpha, B=@n38.Alpha, Alpha=@n33.Alpha) {desc=Use top left height lerp}
  FunctionOutput(None=LinearInterpolate(
    A=LinearInterpolate(A=@n11, B=@n13, Alpha=@n25) {desc=Bottom Corners}
    B=LinearInterpolate(A=@n27, B=@n29, Alpha=@n34) {desc=Top Corners}
    Alpha=StaticSwitch(True=@n39, False=@n23.G, Value=@n24)
   ) {desc=Vertical Transition})
  n40 = Min(A=@n22)
  LinearInterpolate(A=@n30.G, B=@n38.Alpha, Alpha=@n40)
  LinearInterpolate(A=@n35.G, B=@n37.Alpha, Alpha=@n40)
  LinearInterpolate()
  LinearInterpolate(A=@n30.R, B=@n33.Alpha, Alpha=@n40)
  MaterialFunctionCall() {materialFunction=&ScaleUVsByCenter}
  Multiply()
  ScalarParameter()
  n41 = Multiply(A=LinearInterpolate(), B=Multiply() {constB=4})
  n42 = LinearInterpolate(A=MaterialFunctionCall(XY=@n41) {materialFunction=&Swizzle}.YX, B=@n41)
  n43 = Subtract(B=Divide() {constB=4}) {constB=0}
  TextureSample(UVs=Add(A=Multiply(A=@n1), B=Subtract(A=@n42, B=@n43) {constB=0}), "DDX(UVs)"=DDX(Value=@n42), "DDY(UVs)"=DDY(Value=@n42)) {desc=Top Right, mipValueMode=TMVM_Derivative}
 ]
}
