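# Library material function that converts a vector displacement map baked in xNormal into a displacement vector.
# It remaps the Vector input (non-sRGB map RGB) from 0..1 to -1..1 for direction.
# The ScalarDisplacement input (map alpha) interpolates between the Min and Max scalars to set magnitude.
# The product of both passes through a Transform node at its default spaces to the single output.
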MaterialFunction /Engine/Functions/Engine_MaterialFunctions01/Texturing/VectorDisplacement
graph {
 properties {bExposeToLibrary=true, description=Use a Baked Vector displacement map from Xnormal, libraryCategoriesText [Misc,Texturing]}
 nodes [
  n1 = FunctionInput() {
   desc="RGB of Vector displacement map, should have sRGB disabled on texture properties"
   description="RGB of Vector displacement map, should have sRGB disabled on texture properties"
   inputName=Vector
  }
  n2 = LinearInterpolate(A=Constant3Vector() {constant=<rgba -1 -1 -1 0>}, B=Constant3Vector() {constant=<rgba 1 1 1 0>}, Alpha=@n1)
  n3 = FunctionInput() {inputName=Min, inputType=FunctionInput_Scalar, sortPriority=2}
  n4 = FunctionInput() {inputName=Max, inputType=FunctionInput_Scalar, sortPriority=3, previewValue=<1 0 0 1>}
  n5 = LinearInterpolate(
   A=AppendVector(A=AppendVector(A=@n3, B=@n3), B=@n3)
   B=AppendVector(A=AppendVector(A=@n4, B=@n4), B=@n4)
   Alpha=FunctionInput() {desc=alpha of displacement map, description=alpha of displacement map, inputName=ScalarDisplacement, sortPriority=1}
  ) {constA=0.14, constB=7.27}
  FunctionOutput(None=Transform(None=Multiply(A=@n2, B=@n5)))
 ]
}
