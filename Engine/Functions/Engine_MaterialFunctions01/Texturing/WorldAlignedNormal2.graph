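# WorldAlignedNormal2 is a material function that tiles a normal map in world space using triplanar projection.
# World position divided by TextureSize drives three planar samples, whose channels are reoriented into world axes.
# The Normal input, transformed by its X and Z magnitudes, blends the projections.
# Half the vertex normal is added before conversion to tangent space, yielding XYZ, side-only XY, top-only Z and renormalized XYZFlatTop outputs.
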MaterialFunction /Engine/Functions/Engine_MaterialFunctions01/Texturing/WorldAlignedNormal2
graph {
 properties {description=Tiles a texture in worldspace, libraryCategoriesText [Misc,Texturing]}
 nodes [
  n1 = Multiply(A=VertexNormalWS(), B=Constant() {r=0.5})
  n2 = FunctionInput() {bUsePreviewValueAsDefault=true, description=Size of texture in worldspace units, inputName=TextureSize, previewValue=<64 64 64 6>}
  n3 = Divide(A=WorldPosition() {worldPositionShaderOffset=WPT_ExcludeAllShaderOffsets}.XYZ, B=Multiply(A=@n2) {constB=-1})
  n4 = FunctionInput(Preview=TextureObject() {samplerType=SAMPLERTYPE_Normal, texture=/Engine/EngineProduction/ReferenceNormalMap}) {inputName=TextureObject, inputType=FunctionInput_Texture2D}
  n5 = TextureSample(UVs=ComponentMask(None=@n3) {b=true, r=true}, Tex=@n4) {samplerSource=SSM_Wrap_WorldGroupSettings, samplerType=SAMPLERTYPE_Normal}
  n6 = Multiply(A=@n5.RGB) {constB=-1}
  n7 = ComponentMask(None=@n6) {r=true}
  n8 = Transform(None=FunctionInput() {bUsePreviewValueAsDefault=true, inputName=Normal, previewValue=<0 0 1 1>})
  n9 = Multiply(A=@n5.B, B=ComponentMask(None=@n8) {g=true})
  n10 = ComponentMask(None=@n8) {r=true}
  n11 = TextureSample(UVs=ComponentMask(None=@n3) {b=true, g=true}, Tex=@n4) {samplerSource=SSM_Wrap_WorldGroupSettings, samplerType=SAMPLERTYPE_Normal}
  n12 = Multiply(A=@n10, B=@n11.B)
  n13 = Multiply(A=@n11.RGB) {constB=-1}
  n14 = ComponentMask(None=@n13) {r=true}
  n15 = LinearInterpolate(
   A=AppendVector(A=AppendVector(A=@n7, B=@n9), B=ComponentMask(None=@n6) {g=true})
   B=AppendVector(A=AppendVector(A=@n12, B=@n14), B=ComponentMask(None=@n13) {g=true})
   Alpha=Abs(None=@n10)
  )
  n16 = TextureSample(UVs=ComponentMask(None=@n3) {g=true, r=true}, Tex=@n4) {samplerSource=SSM_Wrap_WorldGroupSettings, samplerType=SAMPLERTYPE_Normal}
  n17 = Constant3Vector() {constant=<rgba -1 -1 1 0>}
  n18 = ComponentMask(None=@n8) {b=true}
  n19 = AppendVector(A=ComponentMask(None=Multiply(A=@n16.RGB, B=@n17) {constB=-1}) {g=true, r=true}, B=Multiply(A=@n16.B, B=@n18))
  n20 = Normalize(VectorInput=LinearInterpolate(A=@n15, B=@n19, Alpha=Abs(None=@n18)))
  n21 = Transform(None=Normalize(VectorInput=Add(A=@n1, B=@n20))) {transformSourceType=TRANSFORMSOURCE_World, transformType=TRANSFORM_Tangent}
  FunctionOutput(None=@n21) {outputName=XYZ Texture, sortPriority=1}
  n22 = Normalize(VectorInput=@n15)
  FunctionOutput(None=Transform(None=Add(A=@n1, B=@n22)) {transformSourceType=TRANSFORMSOURCE_World, transformType=TRANSFORM_Tangent}) {outputName=XY Texture, sortPriority=1}
  n23 = Normalize(VectorInput=@n19)
  FunctionOutput(None=Transform(None=Add(A=@n1, B=@n23)) {transformSourceType=TRANSFORMSOURCE_World, transformType=TRANSFORM_Tangent}) {outputName=Z Texture, sortPriority=1}
  FunctionOutput(None=Normalize(VectorInput=@n21)) {outputName=XYZFlatTop, sortPriority=1}
 ]
}
