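# BitMask is a library material function under Misc and Texturing that produces a 1-bit mask from a grayscale image and a 0-255 value.
# It divides the Bit input by 255 and compares it to the BitMask sample in an If node.
# The output is 1 where the two are equal and 0 otherwise, so each gray level acts as its own selectable mask.
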MaterialFunction /Engine/Functions/Engine_MaterialFunctions01/Texturing/BitMask
graph {
 properties {
  bExposeToLibrary=true
  description=Provides a 1bit mask from a supplied Grayscale image and a value between 0-255
  libraryCategoriesText [Misc,Texturing]
 }
 nodes [
  n1 = FunctionInput() {description=Value between 0-255, inputName=Bit, inputType=FunctionInput_Scalar}
  n2 = Constant() {r=255}
  n3 = Constant()
  FunctionOutput(None=If(
    A=FunctionInput() {description=Should be texture using TC_Grayscale, inputName=BitMask, inputType=FunctionInput_Scalar}
    B=Divide(A=@n1, B=@n2)
    "A > B"=@n3
    "A == B"=Constant() {r=1}
    "A < B"=@n3
   ))
 ]
}
