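# Material function in the engine library that converts a camera vector into lat-long UVs, intended for skybox textures.
# VectorToRadialValue turns the vector's XY into U, while a default bias-scale remaps Z linearly into V, with no inverse trig for latitude.
# The input defaults to the world-space camera vector and accepts a distorted vector for special effects.
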MaterialFunction /Engine/Functions/Engine_MaterialFunctions01/Texturing/CameraVectorToLatLongUV
graph {
 properties {
  bExposeToLibrary=true
  description=Converts camera vector to latlong UVs. Helpful for skybox textures
  libraryCategoriesText [Coordinates,Texturing]
 }
 nodes [
  n1 = FunctionInput(Preview=CameraVectorWS()) {bUsePreviewValueAsDefault=true, desc=Allows you to distort camera vector for special effects, inputName=CameraVector}
  n2 = MaterialFunctionCall(Vector or UVs=ComponentMask(None=@n1) {g=true, r=true}) {materialFunction=&VectorToRadialValue}
  n3 = ComponentMask(None=@n1) {b=true}
  FunctionOutput(None=AppendVector(A=ComponentMask(None=@n2.Radial Coordinates) {r=true}, B=ConstantBiasScale(None=@n3))) {outputName=Uvs}
 ]
}
