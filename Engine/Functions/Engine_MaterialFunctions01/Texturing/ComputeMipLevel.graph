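# ComputeMipLevel is a library material function that computes a texture mip level from UVs and a texture size in pixels.
# It scales the screen-space UV derivatives (DDX, DDY) by the size and takes the squared length of each.
# An If node selects the larger value, then a square root and a Custom log2 node produce the MipLevel output, with no clamping.
# Both inputs have preview defaults, including a 2048 size.
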MaterialFunction /Engine/Functions/Engine_MaterialFunctions01/Texturing/ComputeMipLevel
graph {
 properties {bExposeToLibrary=true}
 nodes [
  n1 = FunctionInput(Preview=TextureCoordinate()) {bUsePreviewValueAsDefault=true, description=UVs to compute Mips from, inputName=UVs, inputType=FunctionInput_Vector2}
  n2 = FunctionInput(Preview=Constant() {r=2048}) {
   bUsePreviewValueAsDefault=true
   description=Custom Mip math needs to know the size of your texture in Pixels
   inputName=Texture Size
   inputType=FunctionInput_Vector2
  }
  n3 = ComponentMask(None=@n2) {g=true, r=true}
  n4 = Multiply(A=DDX(Value=@n1), B=@n3)
  n5 = DotProduct(A=@n4, B=@n4)
  n6 = Multiply(A=DDY(Value=@n1), B=@n3)
  n7 = DotProduct(A=@n6, B=@n6)
  MaterialExpressionSquareRoot_11 = SquareRoot(None=If(A=@n5, B=@n7, "A > B"=@n5, "A < B"=@n7) {desc=Max})
  n8 = Custom!(X=@MaterialExpressionSquareRoot_11) {
   code="log2(X)"
   outputType=CMOT_Float1
   inputs [{inputName=X, input {expression=&self:MaterialExpressionSquareRoot_11}}]
  }
  FunctionOutput(None=@n8) {outputName=MipLevel}
 ]
}
