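# Library material function that derives a secondary tangent basis for another UV set, defaulting to UV channel 1, and transforms an input normal into world space with it.
# Screen-space derivatives of world position (excluding shader offsets) and UVs combine with the vertex normal via cross products to build tangent and binormal.
# Both share one inverse-max-length scale, preserving their relative lengths.
# Outputs tangent, binormal and world-space normal, using vertex normal as Z axis.
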MaterialFunction /Engine/Functions/Engine_MaterialFunctions01/Texturing/DeriveTangentBasis
graph {
 properties {bExposeToLibrary=true, description=Derive a secondary tangent basis from another UV channel.}
 nodes [
  n1 = Constant3Vector() {constant=<rgba 0 0 1 0>}
  n2 = WorldPosition() {worldPositionShaderOffset=WPT_ExcludeAllShaderOffsets}
  n3 = VertexNormalWS()
  n4 = CrossProduct(A=DDY(Value=@n2.XYZ) {desc=dp2}, B=@n3) {desc=dp2perp}
  n5 = FunctionInput(Preview=TextureCoordinate() {coordinateIndex=1}) {
   bUsePreviewValueAsDefault=true
   description=Which UV set to use when creating new tangent basis.
   inputName=UVs
   inputType=FunctionInput_Vector2
  }
  n6 = MaterialFunctionCall(Float2=DDX(Value=@n5) {desc=duv1}) {materialFunction=&BreakOutFloat2Components}
  n7 = CrossProduct(A=@n3, B=DDX(Value=@n2.XYZ) {desc=dp1}) {desc=dp1perp}
  n8 = MaterialFunctionCall(Float2=DDY(Value=@n5) {desc=duv2}) {materialFunction=&BreakOutFloat2Components}
  n9 = Add(A=Multiply(A=@n4, B=@n6.R), B=Multiply(A=@n7, B=@n8.R)) {desc=T}
  n10 = DotProduct(A=@n9, B=@n9)
  n11 = Add(A=Multiply(A=@n4, B=@n6.G), B=Multiply(A=@n7, B=@n8.G)) {desc=B}
  n12 = DotProduct(A=@n11, B=@n11)
  n13 = Divide(B=SquareRoot(None=Max(A=@n10, B=@n12)))
  n14 = Multiply(A=@n9, B=@n13) {desc=Adjusted T}
  n15 = Multiply(A=@n13, B=@n11) {desc=Adjusted T}
  FunctionOutput(None=MaterialFunctionCall(
    VectorToTransform=FunctionInput(Preview=@n1) {inputName=Normal, sortPriority=1}
    BasisX=FunctionOutput(None=@n14) {outputName=Tangent}
    BasisY=FunctionOutput(None=@n15) {outputName=Binormal, sortPriority=1}
    BasisZ=VertexNormalWS()
   ) {materialFunction=&Transform3x3Matrix}.Transformed Vector) {outputName=World Space Normal, sortPriority=2}
 ]
}
