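# Material function FlattenNormal, exposed to the library under Misc and Texturing, which reduces the strength of a normal map.
# It linearly interpolates the Normal input toward a constant flat vector (0,0,1), using the scalar Flatness input as alpha.
# Both inputs use their preview values as defaults, and the result is output without renormalization.
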MaterialFunction /Engine/Functions/Engine_MaterialFunctions01/Texturing/FlattenNormal
graph {
 properties {bExposeToLibrary=true, description="Lerps a NormalMap with 0,0,1", libraryCategoriesText [Misc,Texturing]}
 nodes [FunctionOutput(None=LinearInterpolate(
    A=FunctionInput() {bUsePreviewValueAsDefault=true, inputName=Normal, previewValue=<0 0 1 1>}
    B=Constant3Vector() {constant=<rgba 0 0 1 0>}
    Alpha=FunctionInput() {bUsePreviewValueAsDefault=true, inputName=Flatness, inputType=FunctionInput_Scalar, sortPriority=1}
   ))]
}
