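# Material function exposed to the library that returns point-sampled UVs, forcing nearest-neighbor texture sampling.
# It multiplies UVs (defaulting to TexCoord) by a Texture Size input, floors to whole texels, adds a half-texel offset, then divides by the size to land on texel centers.
# Its description limits use to forced mip levels and special cases like writing non-power-of-two render targets.
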MaterialFunction /Engine/Functions/Engine_MaterialFunctions01/Texturing/PointSampledUVs
graph {
 properties {
  bExposeToLibrary=true
  description="Returns UVs that are point sampled which is a way to force nearest neighbor sampling. This should only be used with a forced mip level and only in situations that require it (such as writing render targets in non power of 2 sizes)."
 }
 nodes [
  n1 = TextureCoordinate()
  n2 = FunctionInput() {bUsePreviewValueAsDefault=true, inputName=Texture Size, inputType=FunctionInput_Vector2, sortPriority=1, previewValue=<256 256 0 1>}
  n3 = Floor(None=Multiply(A=FunctionInput(Preview=@n1) {bUsePreviewValueAsDefault=true, inputName=UVs, inputType=FunctionInput_Vector2}, B=@n2))
  FunctionOutput(None=Divide(A=Add(A=@n3) {constB=0.5}, B=@n2))
 ]
}
