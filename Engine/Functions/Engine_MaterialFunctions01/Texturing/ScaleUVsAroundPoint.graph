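# Material function that scales UVs around a chosen center point, exposed to the material library.
# It divides UVs by the scale, then re-offsets them to keep the center fixed.
# The center comes from pixel XY divided by a single texture size in Z, so only square textures are supported.
# A second output multiplies per-axis SphereMasks to mask the scaled UVs' 0-1 square.
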MaterialFunction /Engine/Functions/Engine_MaterialFunctions01/Texturing/ScaleUVsAroundPoint
graph {
 properties {bExposeToLibrary=true}
 nodes [
  n1 = FunctionInput(Preview=TextureCoordinate()) {
   bUsePreviewValueAsDefault=true
   description=UVs. Only need to plug in if you want a different UV channel.
   inputName=UVs
   inputType=FunctionInput_Vector2
  }
  n2 = FunctionInput(Preview=Constant() {r=1}) {description="Scale factor. Default = 1", inputName=Texture Scale, inputType=FunctionInput_Vector2}
  n3 = FunctionInput(Preview=Constant3Vector() {constant=<rgba 1024 1024 2048 0>}) {
   bUsePreviewValueAsDefault=true
   description="X and Y should specify the desired center of scaling in pixels. Z should be the texture size in pixels (ie, 1024).  Currently only supports square textures!"
   inputName=Offset XY Texturesize Z
  }
  n4 = Divide(A=ComponentMask(None=@n3) {g=true, r=true}, B=ComponentMask(None=@n3) {b=true})
  n5 = Subtract(A=Add(A=Divide(A=@n1, B=@n2), B=@n4), B=Divide(A=@n4, B=@n2) {desc=Scale the center offset by the scaling factor}) {desc=Re offset to maintain center}
  FunctionOutput(None=@n5) {outputName=UVs}
  n6 = Constant() {r=0.5}
  n7 = MaterialFunctionCall(Float2=@n5) {materialFunction=&BreakOutFloat2Components}
  FunctionOutput(None=Multiply(A=SphereMask(A=@n6, B=@n7.R) {attenuationRadius=0.5}, B=SphereMask(A=@n7.G, B=@n6) {attenuationRadius=0.5})) {outputName=0-1 mask, sortPriority=1}
 ]
}
