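# SlopeMask is a library material function that outputs a slope-based mask, described as useful for effects like snow and moss.
# It transforms an input tangent-space normal, dots it with a world direction vector (default +Z) and clamps the result.
# A falloff exponent shapes the mask, then the CheapContrast function sharpens it before output.
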MaterialFunction /Engine/Functions/Engine_MaterialFunctions01/Texturing/SlopeMask
graph {
 properties {bExposeToLibrary=true, description="Slopemask, useful for stuff like snow, moss", libraryCategoriesText [Misc,Texturing]}
 nodes [
  n1 = FunctionInput() {bUsePreviewValueAsDefault=true, inputName=TangentNormal, previewValue=<0 0 1 1>}
  n2 = FunctionInput() {
   bUsePreviewValueAsDefault=true
   description="World direction for slop, typically points down"
   inputName=SlopAngle
   sortPriority=1
   previewValue=<0 0 1 1>
  }
  n3 = Clamp(None=DotProduct(A=Transform(None=@n1), B=@n2))
  n4 = FunctionInput() {bUsePreviewValueAsDefault=true, inputName=FalloffPower, inputType=FunctionInput_Scalar, sortPriority=2, previewValue=<0.5 0 0 1>}
  n5 = FunctionInput() {bUsePreviewValueAsDefault=true, inputName=CheapContrast, inputType=FunctionInput_Scalar, sortPriority=3, previewValue=<0.7 0 0 1>}
  FunctionOutput(None=MaterialFunctionCall(In=Power(Base=@n3, Exp=@n4), Contrast=@n5) {materialFunction=&CheapContrast}.Result)
 ]
}
