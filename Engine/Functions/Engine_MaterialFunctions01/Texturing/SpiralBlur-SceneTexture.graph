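# Material function that spiral-blurs the scene color behind the pixel over a set UV distance.
# A Custom HLSL node averages samples across nested distance and radial loops, rotating each ring by Radial Offset, scaling by Distance Mask and clamping to viewport bounds.
# Temporal AA parameters and tiling noise jitter rotation and distance.
# A second SceneColor output, passed through Max, must feed the opacity chain for the scene texture to work.
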MaterialFunction /Engine/Functions/Engine_MaterialFunctions01/Texturing/SpiralBlur-SceneTexture
graph {
 properties {
  bExposeToLibrary=true
  description=Performs a Spiral Blur with controllable number of steps and offests.
  libraryCategoriesText [Misc,Misc,Texturing]
 }
 nodes [
  MaterialExpressionFunctionInput_14 = FunctionInput() {
   bUsePreviewValueAsDefault=true
   description=Total Distance of Blur in UV Space
   inputName=Distance
   inputType=FunctionInput_Scalar
   previewValue=<0.05 0 0 1>
  }
  MaterialExpressionFunctionInput_17 = FunctionInput() {
   bUsePreviewValueAsDefault=true
   description=Number of steps along distance
   inputName=Distance Steps
   inputType=FunctionInput_Scalar
   sortPriority=1
   previewValue=<16 0 0 1>
  }
  MaterialExpressionFunctionInput_1 = FunctionInput() {
   bUsePreviewValueAsDefault=true
   description=Number of radial steps to perform
   inputName=Radial Steps
   inputType=FunctionInput_Scalar
   sortPriority=2
   previewValue=<8 0 0 1>
  }
  MaterialExpressionFunctionInput_9 = FunctionInput() {
   bUsePreviewValueAsDefault=true
   description=Number of radial steps to perform
   inputName=TempAA Radial Blur
   inputType=FunctionInput_Scalar
   sortPriority=4
   previewValue=<1 0 0 1>
  }
  MaterialExpressionFunctionInput_13 = FunctionInput() {
   bUsePreviewValueAsDefault=true
   description=Supply a mask for objects closer to the surface to receive less blur
   inputName=Distance Mask
   inputType=FunctionInput_Scalar
   sortPriority=6
   previewValue=<1 0 0 1>
  }
  MaterialExpressionTextureObject_5 = TextureObject() {samplerType=SAMPLERTYPE_LinearGrayscale, texture=/Engine/EngineMaterials/Good64x64TilingNoiseHighFreq}
  MaterialExpressionFunctionInput_12 = FunctionInput() {
   bUsePreviewValueAsDefault=true
   description=Number of radial steps to perform
   inputName=TempAA Distance Blur
   inputType=FunctionInput_Scalar
   sortPriority=5
   previewValue=<1 0 0 1>
  }
  MaterialExpressionFunctionInput_4 = FunctionInput() {
   bUsePreviewValueAsDefault=true
   description=Number of radial steps to perform
   inputName=Radial Offset
   inputType=FunctionInput_Scalar
   sortPriority=3
   previewValue=<0.618 0 0 1>
  }
  n1 = Custom!(
   Distance=@MaterialExpressionFunctionInput_14
   DistanceSteps=@MaterialExpressionFunctionInput_17
   RadialSteps=@MaterialExpressionFunctionInput_1
   TempAARotation=@MaterialExpressionFunctionInput_9
   DistanceMask=@MaterialExpressionFunctionInput_13
   Tex=@MaterialExpressionTextureObject_5
   TempAADistance=@MaterialExpressionFunctionInput_12
   RadialOffset=@MaterialExpressionFunctionInput_4
  ) {
   code="float3 CurColor = 0;\r\nfloat2 BaseUV = MaterialFloat2(ScreenAlignedPosition(Parameters.ScreenPosition).xy);\r\nfloat2 NewUV = BaseUV;\r\nfloat StepSize = Distance / (int) DistanceSteps;\r\nfloat CurDistance = 0;\r\nfloat2 CurOffset = 0;\r\nfloat TwoPi = 6.283185;\r\nfloat Substep = 0;\r\nfloat2 ScenePixels=View.BufferSizeAndInvSize.xy*BaseUV;\r\nScenePixels+=View.TemporalAAParams.r;\r\nfloat2 RandomSamp = ((uint)(ScenePixels.x) + 2 * (uint)(ScenePixels.y)) % 5;\r\nRandomSamp+=Texture2DSample(Tex,TexSampler,ScenePixels);\r\nRandomSamp/=5;\r\nRandomSamp-=0.5;\r\nTempAARotation*=RandomSamp;\r\nTempAADistance*=StepSize*RandomSamp;\r\n\r\n\r\nint i=0;\r\nif (DistanceSteps<1)\r\n{\r\nreturn DecodeSceneColorForMaterialNode(NewUV);\r\n}\r\nelse\r\n{\r\nfloat2 BilinearOffset = View.BufferSizeAndInvSize.zw * 0.5f;\r\nfloat2 BilinearMinBufferUV = ViewportUVToBufferUV(float2(0.0f, 0.0f)) + BilinearOffset;\r\nfloat2 BilinearMaxBufferUV = ViewportUVToBufferUV(float2(1.0f, 1.0f)) - BilinearOffset;\r\n//CurDistance += 0.5*StepSize;\r\nwhile ( i < (int) DistanceSteps)\r\n{\r\n\r\n//CurDistance+=StepSize;\r\nfor (int j = 0; j < (int) RadialSteps; j++)\r\n{\r\nCurOffset.x = cos(TwoPi*((TempAARotation+Substep) / RadialSteps));\r\nCurOffset.y = sin(TwoPi*((TempAARotation+Substep) / RadialSteps));\r\nCurOffset *=DistanceMask;\r\nNewUV.x = BaseUV.x + (CurOffset.x * (CurDistance+(RandomSamp*TempAADistance)));\r\nNewUV.y = BaseUV.y + (CurOffset.y * (CurDistance+(RandomSamp*TempAADistance)));\r\nNewUV = clamp(NewUV, BilinearMinBufferUV, BilinearMaxBufferUV);\r\nCurColor += DecodeSceneColorForMaterialNode(NewUV);\r\n//CurDistance+=(StepSize+(TempAADistance))/RadialSteps;\r\nSubstep++;\r\n}\r\nCurDistance+=StepSize;\r\nSubstep+=RadialOffset;\r\ni++;\r\n}\r\nCurColor = CurColor / ((int)DistanceSteps*(int)RadialSteps);\r\nreturn CurColor;\r\n}"
   desc="float3 CurColor = 0;\r\nfloat2 BaseUV = MaterialFloat2(ScreenAlignedPosition(Parameters.ScreenPosition).xy);\r\nfloat2 NewUV = BaseUV;\r\nfloat StepSize = Distance / (int) DistanceSteps;\r\nfloat CurDistance = 0;\r\nfloat2 CurOffset = 0;\r\nfloat TwoPi = 6.283185;\r\nfloat Substep = 0;\r\nfloat2 ScenePixels=View.BufferSizeAndInvSize.xy*BaseUV;\r\nScenePixels+=View.TemporalAAParams.r;\r\nfloat2 RandomSamp = ((uint)(ScenePixels.x) + 2 * (uint)(ScenePixels.y)) % 5;\r\nRandomSamp+=Texture2DSample(Tex,TexSampler,ScenePixels);\r\nRandomSamp/=5;\r\nRandomSamp-=0.5;\r\nTempAARotation*=RandomSamp;\r\nTempAADistance*=StepSize*RandomSamp;\r\n\r\n\r\nint i=0;\r\nif (DistanceSteps<1)\r\n{\r\nreturn DecodeSceneColorForMaterialNode(NewUV);\r\n}\r\nelse\r\n{\r\nfloat2 BilinearOffset = View.BufferSizeAndInvSize.zw * 0.5f;\r\nfloat2 BilinearMinBufferUV = ViewportUVToBufferUV(float2(0.0f, 0.0f)) + BilinearOffset;\r\nfloat2 BilinearMaxBufferUV = ViewportUVToBufferUV(float2(1.0f, 1.0f)) - BilinearOffset;\r\n//CurDistance += 0.5*StepSize;\r\nwhile ( i < (int) DistanceSteps)\r\n{\r\n\r\n//CurDistance+=StepSize;\r\nfor (int j = 0; j < (int) RadialSteps; j++)\r\n{\r\nCurOffset.x = cos(TwoPi*((TempAARotation+Substep) / RadialSteps));\r\nCurOffset.y = sin(TwoPi*((TempAARotation+Substep) / RadialSteps));\r\nCurOffset *=DistanceMask;\r\nNewUV.x = BaseUV.x + (CurOffset.x * (CurDistance+(RandomSamp*TempAADistance)));\r\nNewUV.y = BaseUV.y + (CurOffset.y * (CurDistance+(RandomSamp*TempAADistance)));\r\nNewUV = clamp(NewUV, BilinearMinBufferUV, BilinearMaxBufferUV);\r\nCurColor += DecodeSceneColorForMaterialNode(NewUV);\r\n//CurDistance+=(StepSize+(TempAADistance))/RadialSteps;\r\nSubstep++;\r\n}\r\nCurDistance+=StepSize;\r\nSubstep+=RadialOffset;\r\ni++;\r\n}\r\nCurColor = CurColor / ((int)DistanceSteps*(int)RadialSteps);\r\nreturn CurColor;\r\n}"
   inputs [
    {
     inputName=Distance
     input {expression=&self:MaterialExpressionFunctionInput_14}
    }
    {
     inputName=DistanceSteps
     input {expression=&self:MaterialExpressionFunctionInput_17}
    }
    {
     inputName=RadialSteps
     input {expression=&self:MaterialExpressionFunctionInput_1}
    }
    {
     inputName=TempAARotation
     input {expression=&self:MaterialExpressionFunctionInput_9}
    }
    {
     inputName=DistanceMask
     input {expression=&self:MaterialExpressionFunctionInput_13}
    }
    {
     inputName=Tex
     input {expression=&self:MaterialExpressionTextureObject_5}
    }
    {
     inputName=TempAADistance
     input {expression=&self:MaterialExpressionFunctionInput_12}
    }
    {
     inputName=RadialOffset
     input {expression=&self:MaterialExpressionFunctionInput_4}
    }
   ]
  }
  FunctionOutput(None=@n1)
  FunctionOutput(None=Max(A=SceneColor!(UVs).RGB)) {
   description=You will need to hook this up to your opacity chain for the scenetexture to work. Since it has been clamped to 0 you should simply Add this to the rest of your opacity chain
   outputName=SceneColor clamp to 0
   sortPriority=1
  }
 ]
}
