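# Material function that performs a spiral blur on a supplied Texture2D, exposed to the library under Misc and Texturing.
# A single Custom HLSL node loops over distance rings and radial steps around the UV, rotating each ring by Radial Offset.
# It weights samples by distance raised to Kernel Power, favoring outer rings at positive powers, and normalizes the sum.
# Distance Steps below one returns an unblurred sample.
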MaterialFunction /Engine/Functions/Engine_MaterialFunctions01/Texturing/SpiralBlur-Texture
graph {
 properties {
  bExposeToLibrary=true
  description=Performs a Spiral Blur with controllable number of steps and offests.
  libraryCategoriesText [Misc,Misc,Texturing]
 }
 nodes [
  MaterialExpressionFunctionInput_5 = FunctionInput(Preview=TextureObject() {texture=/Engine/EditorMaterials/TargetIcon}) {description=Supply the Texture to Blur as a TextureObject node, inputName=TextureObject, inputType=FunctionInput_Texture2D}
  MaterialExpressionFunctionInput_8 = FunctionInput(Preview=TextureCoordinate()) {bUsePreviewValueAsDefault=true, description=UVs for blurring, inputName=UV, inputType=FunctionInput_Vector2, sortPriority=1}
  MaterialExpressionFunctionInput_14 = FunctionInput() {
   bUsePreviewValueAsDefault=true
   description=Total Distance of Blur in UV Space
   inputName=Distance
   inputType=FunctionInput_Scalar
   sortPriority=3
   previewValue=<0.05 0 0 1>
  }
  MaterialExpressionFunctionInput_17 = FunctionInput() {
   bUsePreviewValueAsDefault=true
   description=Number of steps along distance
   inputName=Distance Steps
   inputType=FunctionInput_Scalar
   sortPriority=4
   previewValue=<16 0 0 1>
  }
  MaterialExpressionFunctionInput_1 = FunctionInput() {
   bUsePreviewValueAsDefault=true
   description=Number of radial steps to perform
   inputName=Radial Steps
   inputType=FunctionInput_Scalar
   sortPriority=5
   previewValue=<8 0 0 1>
  }
  MaterialExpressionFunctionInput_2 = FunctionInput() {
   bUsePreviewValueAsDefault=true
   description=Number of radial steps to perform
   inputName=Radial Offset
   inputType=FunctionInput_Scalar
   sortPriority=6
   previewValue=<0.618 0 0 1>
  }
  MaterialExpressionFunctionInput_3 = FunctionInput() {
   bUsePreviewValueAsDefault=true
   description=Number of radial steps to perform
   inputName=Kernel Power
   inputType=FunctionInput_Scalar
   sortPriority=7
   previewValue=<1 0 0 1>
  }
  n1 = Custom!(
   Tex=@MaterialExpressionFunctionInput_5
   UV=@MaterialExpressionFunctionInput_8
   Distance=@MaterialExpressionFunctionInput_14
   DistanceSteps=@MaterialExpressionFunctionInput_17
   RadialSteps=@MaterialExpressionFunctionInput_1
   RadialOffset=@MaterialExpressionFunctionInput_2
   KernelPower=@MaterialExpressionFunctionInput_3
  ) {
   code="float3 CurColor=0;\r\nfloat2 NewUV = UV;\r\nint i=0;\r\nfloat StepSize = Distance / (int) DistanceSteps;\r\nfloat CurDistance=0;\r\nfloat2 CurOffset=0;\r\nfloat SubOffset = 0;\r\nfloat TwoPi = 6.283185;\r\nfloat accumdist=0;\r\n\r\nif (DistanceSteps < 1)\r\n\r\n{\r\nreturn Texture2DSample(Tex,TexSampler,UV);\t\t\r\n}\r\nelse\r\n{\r\nwhile ( i < (int) DistanceSteps)\r\n{\r\n\tCurDistance += StepSize;\r\n\tfor (int j = 0; j < (int) RadialSteps; j++)\r\n\t{\r\nSubOffset +=1;\r\nCurOffset.x = cos(TwoPi*(SubOffset / RadialSteps));\r\nCurOffset.y = sin(TwoPi*(SubOffset / RadialSteps));\r\n\t\tNewUV.x = UV.x + CurOffset.x * CurDistance;\r\n\t\tNewUV.y = UV.y + CurOffset.y * CurDistance;\r\n\t\tfloat distpow = pow(CurDistance, KernelPower);\r\n\t\tCurColor += Texture2DSample(Tex,TexSampler,NewUV)*distpow;\t\t\r\n\t\taccumdist += distpow;\r\n\t}\r\nSubOffset +=RadialOffset;\r\ni++;\r\n}\r\nCurColor = CurColor;\r\nCurColor /=accumdist;\r\nreturn CurColor;\r\n}"
   inputs [
    {
     inputName=Tex
     input {expression=&self:MaterialExpressionFunctionInput_5}
    }
    {inputName=UV, input {expression=&self:MaterialExpressionFunctionInput_8}}
    {
     inputName=Distance
     input {expression=&self:MaterialExpressionFunctionInput_14}
    }
    {
     inputName=DistanceSteps
     input {expression=&self:MaterialExpressionFunctionInput_17}
    }
    {
     inputName=RadialSteps
     input {expression=&self:MaterialExpressionFunctionInput_1}
    }
    {
     inputName=RadialOffset
     input {expression=&self:MaterialExpressionFunctionInput_2}
    }
    {
     inputName=KernelPower
     input {expression=&self:MaterialExpressionFunctionInput_3}
    }
   ]
  }
  FunctionOutput(None=@n1)
 ]
}
