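# UVCropping is a library material function that builds a rectangular mask from crop amounts for left, right, top and bottom, meant to be multiplied with a texture.
# Four If nodes compare the texture coordinate's U and V against each edge, with right and bottom measured from one, each giving one inside and zero outside.
# Their product forms the hard-edged crop mask output.
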MaterialFunction /Engine/Functions/Engine_MaterialFunctions01/Texturing/UVCropping
graph {
 properties {bExposeToLibrary=true, description=Creates a cropped region to be multiplied to texture, libraryCategoriesText [Texturing]}
 nodes [
  n1 = MaterialFunctionCall(Float2=TextureCoordinate()) {materialFunction=&BreakOutFloat2Components}
  n2 = Constant() {r=1}
  n3 = Constant()
  n4 = If(A=@n1.R, B=FunctionInput() {inputName=Crop Left, inputType=FunctionInput_Scalar}, "A > B"=@n2, "A < B"=@n3)
  n5 = If(A=@n1.R, B=Subtract(A=Constant() {r=1}, B=FunctionInput() {inputName=Crop Right, inputType=FunctionInput_Scalar}), "A > B"=@n3, "A < B"=@n2)
  n6 = Constant()
  n7 = Constant() {r=1}
  n8 = If(A=@n1.G, B=Subtract(A=Constant() {r=1}, B=FunctionInput() {inputName=Crop Bottom, inputType=FunctionInput_Scalar}), "A > B"=@n6, "A < B"=@n7)
  n9 = If(A=@n1.G, B=FunctionInput() {inputName=Crop Top, inputType=FunctionInput_Scalar}, "A > B"=@n7, "A < B"=@n6)
  FunctionOutput(None=Multiply(A=Multiply(A=@n4, B=@n5), B=Multiply(A=@n8, B=@n9)))
 ]
}
