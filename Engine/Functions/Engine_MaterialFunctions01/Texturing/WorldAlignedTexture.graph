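# Material function that tiles a texture in world space using triplanar projection, independent of mesh UVs.
# World position divided by negated absolute TextureSize supplies UVs for three axis-plane samples using shared wrap samplers.
# Absolute world-normal X and Z, sharpened by CheapContrast, drive two chained lerps blending the side projections, then the top one.
# It outputs the full blend, sides only and top only, with a static switch for Float 4 export.
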MaterialFunction /Engine/Functions/Engine_MaterialFunctions01/Texturing/WorldAlignedTexture
graph {
 properties {bExposeToLibrary=true, description=Tiles a texture in worldspace, libraryCategoriesText [Misc,Texturing]}
 nodes [
  n1 = FunctionInput() {bUsePreviewValueAsDefault=true, description=Size of texture in worldspace units, inputName=TextureSize, previewValue=<64 64 64 6>}
  n2 = Divide(
   A=FunctionInput(Preview=WorldPosition() {worldPositionShaderOffset=WPT_ExcludeAllShaderOffsets}.XYZ) {bUsePreviewValueAsDefault=true, inputName=WorldPosition}
   B=Multiply(A=Abs(None=@n1)) {constB=-1}
  )
  n3 = FunctionInput(Preview=TextureObject() {texture=/Engine/EngineMaterials/DefaultDiffuse}) {inputName=TextureObject, inputType=FunctionInput_Texture2D}
  n4 = TextureSample(UVs=ComponentMask(None=@n2) {b=true, r=true}, Tex=@n3) {samplerSource=SSM_Wrap_WorldGroupSettings}
  n5 = FunctionInput(Preview=StaticBool()) {bUsePreviewValueAsDefault=true, description="Default = False", inputName=Export Float 4, inputType=FunctionInput_StaticBool}
  n6 = TextureSample(UVs=ComponentMask(None=@n2) {b=true, g=true}, Tex=@n3) {samplerSource=SSM_Wrap_WorldGroupSettings}
  n7 = FunctionInput(Preview=VertexNormalWS()) {bUsePreviewValueAsDefault=true, inputName=World Space Normal, previewValue=<0 0 1 1>}
  n8 = Abs(None=ComponentMask(None=@n7) {r=true})
  n9 = FunctionInput() {
   bUsePreviewValueAsDefault=true
   inputName=ProjectionTransitionContrast
   inputType=FunctionInput_Scalar
   sortPriority=10
   previewValue=<1 0 0 1>
  }
  n10 = LinearInterpolate(
   A=StaticSwitch(True=AppendVector(A=@n4.RGB, B=@n4.RGB), False=@n4.RGB, Value=@n5)
   B=StaticSwitch(True=AppendVector(A=@n6.RGB, B=@n6.RGB), False=@n6.RGB, Value=@n5)
   Alpha=ComponentMask(None=MaterialFunctionCall(In=@n8, Contrast=@n9) {materialFunction=&CheapContrast}.Result) {r=true}
  ) {desc=Float3}
  n11 = TextureSample(UVs=ComponentMask(None=@n2) {g=true, r=true}, Tex=@n3) {samplerSource=SSM_Wrap_WorldGroupSettings}
  n12 = StaticSwitch(True=AppendVector(A=@n11.RGB, B=@n11.RGB), False=@n11.RGB, Value=@n5)
  n13 = MaterialFunctionCall(In=Abs(None=ComponentMask(None=@n7) {b=true}), Contrast=@n9) {materialFunction=&CheapContrast}
  FunctionOutput(None=LinearInterpolate(A=@n10, B=@n12, Alpha=ComponentMask(None=@n13.Result) {r=true})) {outputName=XYZ Texture, sortPriority=1}
  FunctionOutput(None=@n10) {outputName=XY Texture, sortPriority=1}
  FunctionOutput(None=@n12) {outputName=Z Texture, sortPriority=1}
 ]
}
